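# Control Rig for the F_MED_FuzzyGlam face accessory that poses hair bones during aiming and head-turn emotes.
# Six sequenced branches apply preset transforms, five weighted by spherical pose readers on the head or right upper arm, multiplied by ADS, crouch or emote variables.
# Four of those weights are smoothed by AlphaInterp; the first combines arm and head readers unsmoothed.
# The last branch poses hair by the absolute of one minus RBANEnabledFloat.
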ControlRigBlueprint /BRCosmetics/Characters/Player/Female/Medium/Bodies/F_MED_FuzzyGlam/Meshes/Parts/F_MED_FuzzyGlam_FaceAcc_CtrlRig_GunHair
graphs [
 Multiply_1_ContainedGraph {nodes [
   n1 = RigVMFunctionEntryNode()
   Multiply = &RigVMFunction_MathDoubleMul(A=@n1.A, B=@n1.B)
   RigVMFunctionReturnNode(Result=&RigVMFunction_MathDoubleMul(B=@n1.C, A=@Multiply.Result).Result)
  ]}
 RigVMFunctionLibrary {links [], nodes []}
 RigVMFunction_Sequence_1_ContainedGraph {nodes [
   n1 = RigVMFunctionEntryNode()
   RigVMFunction_Sequence = &RigVMFunction_Sequence(ExecuteContext=@n1.ExecuteContext)
   n2 = &RigVMFunction_Sequence(ExecuteContext=@RigVMFunction_Sequence.B)
   n3 = &RigVMFunction_Sequence(ExecuteContext=@n2.B)
   n4 = &RigVMFunction_Sequence(ExecuteContext=@n3.B)
   n5 = &RigVMFunction_Sequence(ExecuteContext=@n4.B)
   RigVMFunctionReturnNode(A=@RigVMFunction_Sequence.A, B=@n2.A, C=@n3.A, D=@n4.A, E=@n5.A, F=@n5.B)
  ]}
 RigVMModel {nodes [
   RigUnit_BeginExecution = &RigUnit_BeginExecution()
   RigVMFunction_Sequence = RigVMAggregateNode(ExecuteContext=@RigUnit_BeginExecution.ExecutePin)
   n1 = &RigUnit_SphericalPoseReader(
    ActiveRegionScaleFactors=(PositiveWidth=1,NegativeWidth=1,PositiveHeight=1,NegativeHeight=1)
    ActiveRegionSize=0.08
    Debug=(bDrawDebug=false,bDraw2D=False,bDrawLocalAxes=false,DebugScale=25,DebugSegments=20,DebugThickness=0.25)
    DriverAxis=(X=-1,Y=0,Z=0)
    DriverItem=(Type=Bone,Name="upperarm_r")
    FalloffRegionScaleFactors=(PositiveWidth=1,NegativeWidth=1,PositiveHeight=1,NegativeHeight=1)
    FalloffSize=0.12
    FlipHeightScaling=False
    FlipWidthScaling=False
    OptionalParentItem=(Type=Bone,Name="None")
    RotationOffset=(X=60,Y=0,Z=160)
    ExecutePin=@RigVMFunction_Sequence.A
   )
   > n2 = &RigUnit_SphericalPoseReader(
    ActiveRegionScaleFactors=(PositiveWidth=1,NegativeWidth=1,PositiveHeight=1,NegativeHeight=1)
    ActiveRegionSize=0.08
    Debug=(bDrawDebug=false,bDraw2D=False,bDrawLocalAxes=False,DebugScale=25,DebugSegments=20,DebugThickness=0.25)
    DriverAxis=(X=1,Y=0,Z=0)
    DriverItem=(Type=Bone,Name="head")
    FalloffRegionScaleFactors=(PositiveWidth=1,NegativeWidth=1,PositiveHeight=1,NegativeHeight=1)
    FalloffSize=0.12
    FlipHeightScaling=False
    FlipWidthScaling=False
    OptionalParentItem=(Type=Bone,Name="None")
    RotationOffset=(X=0,Y=-85,Z=-30)
   )
   Multiply = RigVMAggregateNode(A=@n1.OutputParam, B=@n2.OutputParam, C=RigVMVariableNode() {variable=isADS}.Value)
   > &RigUnit_ModifyTransforms(
    ItemToModify="((Item=(Type=Bone,Name=\"dyn_r_hair_a_01\"),Transform=(Rotation=(X=-0.1094,Y=0.2346,Z=0.4082,W=0.8754),Translation=(X=0,Y=0,Z=0),Scale3D=(X=1,Y=1,Z=1))),(Item=(Type=Bone,Name=\"dyn_r_hair_a_02\"),Transform=(Rotation=(X=0,Y=0,Z=0,W=1),Translation=(X=-2,Y=0,Z=0) ...(truncated)"
    WeightMinimum=0
    Weight=@Multiply.Result
   )
   n3 = &RigUnit_SphericalPoseReader(
    ActiveRegionScaleFactors=(PositiveWidth=1,NegativeWidth=1,PositiveHeight=1,NegativeHeight=1)
    ActiveRegionSize=0.01
    Debug=(bDrawDebug=false,bDraw2D=False,bDrawLocalAxes=False,DebugScale=25,DebugSegments=20,DebugThickness=0.25)
    DriverAxis=(X=-1,Y=0,Z=0)
    DriverItem=(Type=Bone,Name="upperarm_r")
    FalloffRegionScaleFactors=(PositiveWidth=1,NegativeWidth=1,PositiveHeight=1,NegativeHeight=1)
    FalloffSize=0.09
    FlipHeightScaling=False
    FlipWidthScaling=False
    OptionalParentItem=(Type=Bone,Name="None")
    RotationOffset=(X=-45,Y=0,Z=-40)
    ExecutePin=@RigVMFunction_Sequence.B
   )
   AlphaInterp = &RigVMFunction_AlphaInterp(
    InterpSpeedDecreasing=10
    InterpSpeedIncreasing=10
    Scale=1
    bClampResult=False
    Value=&RigVMFunction_MathDoubleMul(A=@n3.OutputParam, B=RigVMVariableNode() {variable=isADS}.Value).Result
   )
   > &RigUnit_ModifyTransforms(
    ItemToModify="((Item=(Type=Bone,Name=\"dyn_r_hair_a_01\"),Transform=(Rotation=(X=-0.0189,Y=0.0851,Z=0.2156,W=0.9726),Translation=(X=0,Y=0,Z=0),Scale3D=(X=1,Y=1,Z=1))),(Item=(Type=Bone,Name=\"dyn_r_hair_a_02\"),Transform=(Rotation=(X=0,Y=0,Z=0,W=1),Translation=(X=-2,Y=0,Z=0) ...(truncated)"
    WeightMinimum=0
    Weight=@AlphaInterp.Result
   )
   n4 = &RigUnit_SphericalPoseReader(
    ActiveRegionScaleFactors=(PositiveWidth=1,NegativeWidth=1,PositiveHeight=1,NegativeHeight=1)
    ActiveRegionSize=0.02
    Debug=(bDrawDebug=false,bDraw2D=False,bDrawLocalAxes=False,DebugScale=25,DebugSegments=20,DebugThickness=0.25)
    DriverAxis=(X=1,Y=0,Z=0)
    DriverItem=(Type=Bone,Name="head")
    FalloffRegionScaleFactors=(PositiveWidth=1,NegativeWidth=1,PositiveHeight=1,NegativeHeight=1)
    FalloffSize=0.1
    FlipHeightScaling=False
    FlipWidthScaling=False
    OptionalParentItem=(Type=Bone,Name="spine_05")
    RotationOffset=(X=-65,Y=0,Z=-10)
    ExecutePin=@RigVMFunction_Sequence.C
   )
   n5 = &RigVMFunction_AlphaInterp(
    InterpSpeedDecreasing=5
    InterpSpeedIncreasing=5
    Scale=1
    bClampResult=False
    Value=&RigVMFunction_MathDoubleMul(A=@n4.OutputParam, B=RigVMVariableNode() {variable=isADS_Head_Crouch}.Value).Result
   )
   > &RigUnit_ModifyTransforms(
    ItemToModify="((Item=(Type=Bone,Name=\"dyn_r_hair_b_01\"),Transform=(Rotation=(X=-0.0227,Y=-0.1285,Z=-0.1722,W=0.9764),Translation=(X=0,Y=0,Z=0),Scale3D=(X=1,Y=1,Z=1))),(Item=(Type=Bone,Name=\"dyn_m_hair_01\"),Transform=(Rotation=(X=0,Y=0,Z=0,W=1),Translation=(X=0,Y=0,Z=0), ...(truncated)"
    WeightMinimum=0
    Weight=@n5.Result
   )
   n6 = &RigUnit_SphericalPoseReader(
    ActiveRegionScaleFactors=(PositiveWidth=1,NegativeWidth=1,PositiveHeight=1,NegativeHeight=1)
    ActiveRegionSize=0.12
    Debug=(bDrawDebug=True,bDraw2D=False,bDrawLocalAxes=False,DebugScale=25,DebugSegments=20,DebugThickness=0.25)
    DriverAxis=(X=0,Y=1,Z=0)
    DriverItem=(Type=Bone,Name="head")
    FalloffRegionScaleFactors=(PositiveWidth=1,NegativeWidth=1,PositiveHeight=1,NegativeHeight=1)
    FalloffSize=0.25
    FlipHeightScaling=False
    FlipWidthScaling=False
    OptionalParentItem=(Type=Bone,Name="spine_05")
    RotationOffset=(X=20,Y=0,Z=0)
    ExecutePin=@RigVMFunction_Sequence.D
   )
   n7 = &RigVMFunction_AlphaInterp(
    InterpSpeedDecreasing=10
    InterpSpeedIncreasing=10
    Scale=1
    bClampResult=False
    Value=&RigVMFunction_MathDoubleMul(A=@n6.OutputParam, B=RigVMVariableNode() {variable=Emote_HeadRot}.Value).Result
   )
   > &RigUnit_ModifyTransforms(
    ItemToModify="((Item=(Type=Bone,Name=\"dyn_r_hair_b_01\"),Transform=(Rotation=(X=0.0076,Y=0.043,Z=-0.1735,W=0.9839),Translation=(X=0,Y=0,Z=0),Scale3D=(X=1,Y=1,Z=1))),(Item=(Type=Bone,Name=\"dyn_r_hair_a_01\"),Transform=(Rotation=(X=0.0114,Y=-0.0864,Z=0.13,W=0.9877),Translat ...(truncated)"
    WeightMinimum=0
    Weight=@n7.Result
   )
   n8 = &RigUnit_SphericalPoseReader(
    ActiveRegionScaleFactors=(PositiveWidth=1,NegativeWidth=1,PositiveHeight=1,NegativeHeight=1)
    ActiveRegionSize=0.12
    Debug=(bDrawDebug=True,bDraw2D=False,bDrawLocalAxes=False,DebugScale=25,DebugSegments=20,DebugThickness=0.25)
    DriverAxis=(X=0,Y=1,Z=0)
    DriverItem=(Type=Bone,Name="head")
    FalloffRegionScaleFactors=(PositiveWidth=1,NegativeWidth=1,PositiveHeight=1,NegativeHeight=1)
    FalloffSize=0.25
    FlipHeightScaling=False
    FlipWidthScaling=False
    OptionalParentItem=(Type=Bone,Name="spine_05")
    RotationOffset=(X=150,Y=0,Z=0)
    ExecutePin=@RigVMFunction_Sequence.E
   )
   n9 = &RigVMFunction_AlphaInterp(
    InterpSpeedDecreasing=10
    InterpSpeedIncreasing=10
    Scale=1
    bClampResult=False
    Value=&RigVMFunction_MathDoubleMul(A=@n8.OutputParam, B=RigVMVariableNode() {variable=Emote_HeadRot}.Value).Result
   )
   > &RigUnit_ModifyTransforms(
    ItemToModify="((Item=(Type=Bone,Name=\"dyn_l_hair_a_01\"),Transform=(Rotation=(X=0.0038,Y=-0.0435,Z=0.0871,W=0.9952),Translation=(X=0,Y=0,Z=0),Scale3D=(X=1,Y=1,Z=1))),(Item=(Type=Bone,Name=\"dyn_l_hair_a_02\"),Transform=(Rotation=(X=0,Y=0,Z=0,W=1),Translation=(X=-2,Y=0,Z=0) ...(truncated)"
    WeightMinimum=0
    Weight=@n9.Result
   )
   Subtract = &RigVMFunction_MathDoubleSub(A=1, B=RigVMVariableNode() {variable=RBANEnabledFloat}.Value)
   Absolute = &RigVMFunction_MathDoubleAbs(Value=@Subtract.Result)
   &RigUnit_ModifyTransforms(
    ItemToModify="((Item=(Type=Bone,Name=\"dyn_m_hair_base\"),Transform=(Rotation=(X=0,Y=-0.0698,Z=0,W=0.9976),Translation=(X=0,Y=0,Z=0),Scale3D=(X=1,Y=1,Z=1))),(Item=(Type=Bone,Name=\"dyn_l_hair_b_base\"),Transform=(Rotation=(X=0,Y=-0.0698,Z=0,W=0.9976),Translation=(X=0,Y=0,Z= ...(truncated)"
    WeightMinimum=0
    Weight=@Absolute.Result
    ExecutePin=@RigVMFunction_Sequence.F
   )
  ]}
]
hierarchy [
 root {loc=<0 0 0>, rot=<0 0 0 1>}
 pelvis {parent=root, loc=<0 -0.5301 98.0571>, rot=<0.0127 0.707 -0.0127 -0.707>}
 spine_01 {parent=pelvis, loc=<6.1208 -0.9328 0>, rot=<0 0 0.034 -0.9994>}
 spine_02 {parent=spine_01, loc=<6.5997 -0.2095 0>, rot=<0 0 0.0238 -0.9997>}
 spine_03 {parent=spine_02, loc=<6.4914 -0.1237 0>, rot=<0 0 0.0149 -0.9999>}
 spine_04 {parent=spine_03, loc=<6.9598 0.0794 0>, rot=<0 0 -0.0132 -0.9999>}
 spine_05 {parent=spine_04, loc=<6.9987 0 0>, rot=<0 0 -0.0015 -1>}
 neck_01 {parent=spine_05, loc=<11.3138 0.1529 0>, rot=<0 0 -0.2003 -0.9797>}
 neck_02 {parent=neck_01, loc=<4.5627 0 0>, rot=<0 0 0.0504 -0.9987>}
 head {parent=neck_02, loc=<5.3914 0 0>, rot=<0 0 0.115 -0.9934>}
 dyn_r_feather_01 {parent=head, loc=<6.0988 -0.9938 9.2719>, rot=<-0.002 -0.0017 0.1086 -0.9941>}
 dyn_r_feather_02 {parent=dyn_r_feather_01, loc=<5.7325 0 0>, rot=<-0.0006 -0.185 -0.0063 -0.9827>}
 dyn_l_feather_01 {parent=head, loc=<6.0454 -0.9146 -8.8812>, rot=<0.9941 -0.1086 -0.0017 -0.002>}
 dyn_l_feather_02 {parent=dyn_l_feather_01, loc=<5.7325 0 0>, rot=<-0.0006 -0.185 -0.0063 -0.9827>}
 dyn_f_hair {parent=head, loc=<8.1837 8.8982 0.7253>, rot=<0.0957 0.7006 0.7006 0.0957>}
 dyn_r_hair_01 {parent=head, loc=<4.0126 2.4153 6.1565>, rot=<0.0764 0.3814 0.9213 -0.0011>}
 dyn_l_hair_01 {parent=head, loc=<3.7424 2.5611 -5.6943>, rot=<0.1122 -0.9865 0.0125 -0.1182>}
 dyn_l_hair_b_base {parent=head, loc=<6.4875 -6.5082 -4.9445>, rot=<0.0661 -0.8917 0.4425 -0.0682>}
 dyn_l_hair_b_01 {parent=dyn_l_hair_b_base, loc=<0.7613 -0.0285 0.4362>, rot=<0 0 0 1>}
 dyn_l_hair_b_02 {parent=dyn_l_hair_b_01, loc=<7.9067 -0.1331 0.0968>, rot=<0 0 0 1>}
 dyn_l_hair_b_03 {parent=dyn_l_hair_b_02, loc=<6.5143 0.1615 -0.533>, rot=<0 0 0 1>}
 dyn_r_hair_b_base {parent=head, loc=<5.8297 -7.2014 6.5002>, rot=<0.0242 -0.2789 0.96 -0.0089>}
 dyn_r_hair_b_01 {parent=dyn_r_hair_b_base, loc=<0.0285 -0.0124 0.1778>, rot=<0 0 0 1>}
 dyn_r_hair_b_02 {parent=dyn_r_hair_b_01, loc=<8.0142 0.0029 -0.041>, rot=<0 0 0 1>}
 dyn_r_hair_b_03 {parent=dyn_r_hair_b_02, loc=<5.9372 0.0096 -0.1368>, rot=<0 0 0 1>}
 dyn_r_hair_a_base1 {parent=head, loc=<1.5342 -2.1947 7>, rot=<0 0 0.9989 -0.0479>}
 dyn_r_hair_a_01 {parent=dyn_r_hair_a_base1, loc=<0 0 0>, rot=<0 0 0 1>}
 dyn_r_hair_a_02 {parent=dyn_r_hair_a_01, loc=<6.1574 0 0>, rot=<0 0 0 1>}
 dyn_r_hair_a_03 {parent=dyn_r_hair_a_02, loc=<4.0556 0 0>, rot=<0 0 0 1>}
 dyn_l_hair_a_base {parent=head, loc=<0.818 -1.3 -7>, rot=<0.0479 -0.9989 0 0>}
 dyn_l_hair_a_01 {parent=dyn_l_hair_a_base, loc=<0 0 0>, rot=<0 0 0 1>}
 dyn_l_hair_a_02 {parent=dyn_l_hair_a_01, loc=<6.1574 0 0>, rot=<0 0 0 1>}
 dyn_l_hair_a_03 {parent=dyn_l_hair_a_02, loc=<3.5403 0 0>, rot=<0 0 0 1>}
 dyn_m_hair_base {parent=head, loc=<4.8272 -9.5671 0>, rot=<0.0623 -0.7044 0.7044 -0.0623>}
 dyn_m_hair_01 {parent=dyn_m_hair_base, loc=<0 0 0>, rot=<0 0 0 1>}
 dyn_m_hair_02 {parent=dyn_m_hair_01, loc=<7.1428 0 0>, rot=<0 0 0 1>}
 dyn_m_hair_03 {parent=dyn_m_hair_02, loc=<4.5272 0 0>, rot=<0 0 0 1>}
 clavicle_l {parent=spine_05, loc=<5.4831 -0.1053 -3.2946>, rot=<0.0465 -0.7289 -0.009 -0.6829>}
 upperarm_l {parent=clavicle_l, loc=<9.8658 -1.2136 -0.4105>, rot=<0.0164 -0.3245 -0.0269 -0.9454>}
 lowerarm_l {parent=upperarm_l, loc=<26.465 -0.8345 0.259>, rot=<0.0036 0.0161 -0.2298 -0.9731>}
 upperarm_twist_01_l {parent=upperarm_l, loc=<7.2135 -0.1134 0>, rot=<0 0 0 1>}
 clavicle_r {parent=spine_05, loc=<5.4831 -0.1053 3.2946>, rot=<0.7289 0.0465 -0.6829 0.009>}
 upperarm_r {parent=clavicle_r, loc=<-9.8658 1.2136 0.4105>, rot=<0.0164 -0.3245 -0.0269 -0.9454>}
 lowerarm_r {parent=upperarm_r, loc=<-26.465 0.8345 -0.259>, rot=<0.0036 0.0161 -0.2298 -0.9731>}
 upperarm_twist_01_r {parent=upperarm_r, loc=<-7.2135 0.1134 0>, rot=<0 0 0 -1>}
 attach_cape {parent=spine_05, loc=<9.017 -11.714 0>, rot=<0.0616 -0.7044 -0.0616 -0.7044>}
]
variables [
 Emote_HeadRot {type=double, default=0}
 RBANEnabledFloat {type=double, default=0}
 isADS {type=double, default=0}
 isADS_Head_Crouch {type=double, default=0}
]
