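# Physics asset for the FuzzyGlam female character's hair, named as the emote variant, defining capsule bodies for secondary hair motion.
# Three two-bone chains (left, middle, right) hang from the head with limited swing and zero-degree twist limits.
# Upper spine, clavicle, neck and head capsules form the body, linked by free angular constraints.
# Cross-links join neighboring strands at both segments, limited between left and middle but angularly free between right and middle.
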PhysicsAsset /BRCosmetics/Characters/Player/Female/Medium/Bodies/F_MED_FuzzyGlam/Meshes/Parts/F_MED_FuzzyGlam_FaceAcc_Hair_Emote_Physics
bodies {
 clavicle_l [{shapeType=Capsule, length=30, radius=6.125, shapeName=clavicle_l_capsule, center=<5.0495 -2 -1.3288>, rotation=<rot 85 -179.9985 -179.9991>}]
 clavicle_r [{shapeType=Capsule, length=30, radius=6.125, shapeName=clavicle_r_capsule, center=<-5.0495 2 1.3288>, rotation=<rot 85 -179.9985 0.0009>}]
 dyn_l_hair_b_01 [{shapeType=Capsule, length=2.7688, radius=3.5, shapeName=dyn_l_hair_b_01_capsule, center=<3 0 0>, rotation=<rot 90 10.0001 0>}]
 dyn_l_hair_b_02 [{shapeType=Capsule, length=4.2173, radius=4.425, shapeName=dyn_l_hair_b_02_capsule, center=<3.8163 0 0>, rotation=<rot 90 0 0>}]
 dyn_m_hair_01 [{shapeType=Capsule, length=2.7688, radius=3.5, shapeName=dyn_m_hair_01_capsule, center=<3 0 0>, rotation=<rot 90 10 0>}]
 dyn_m_hair_02 [{shapeType=Capsule, length=4.3423, radius=4.175, shapeName=dyn_m_hair_02_capsule, center=<2.8613 0 -0.5>, rotation=<rot 85.0001 -180 -180>}]
 dyn_r_hair_b_01 [{shapeType=Capsule, length=3.0188, radius=3.75, shapeName=dyn_r_hair_b_01_capsule, center=<3 0 0.6>, rotation=<rot 90 4.9996 0>}]
 dyn_r_hair_b_02 [{
   shapeType=Capsule
   length=4.0923
   radius=4.025
   shapeName=dyn_r_hair_b_02_capsule
   center=<4.4199 2.2 -0.2>
   rotation=<rot 65.2333 -90.4755 -106.8644>
  }]
 head [{shapeType=Capsule, length=5, radius=3, shapeName=head_capsule}]
 neck_01 [{shapeType=Capsule, length=5, radius=3, shapeName=neck_01_capsule}]
 neck_02 [{shapeType=Capsule, length=5, radius=3, shapeName=neck_02_capsule}]
 spine_05 [
  {shapeType=Capsule, length=32.8748, radius=6.1565, shapeName=spine_05_capsule, center=<10.3423 4.6906 -0.8074>, rotation=<rot -90 19.9998 0>}
  {shapeType=Capsule, length=43.4972, radius=15, shapeName=spine_05_capsule, center=<-5.4813 -1.1101 0>}
 ]
}
constraints [
 {
  bone1Name=neck_01
  bone2Name=spine_05
  swing1LimitDegrees=45
  swing1Motion=Free
  swing2LimitDegrees=45
  swing2Motion=Free
  twistLimitDegrees=45
  twistMotion=Free
 }
 {
  bone1Name=clavicle_l
  bone2Name=spine_05
  swing1LimitDegrees=45
  swing1Motion=Free
  swing2LimitDegrees=45
  swing2Motion=Free
  twistLimitDegrees=45
  twistMotion=Free
 }
 {
  bone1Name=clavicle_r
  bone2Name=spine_05
  swing1LimitDegrees=45
  swing1Motion=Free
  swing2LimitDegrees=45
  swing2Motion=Free
  twistLimitDegrees=45
  twistMotion=Free
 }
 {
  bone1Name=neck_02
  bone2Name=neck_01
  swing1LimitDegrees=45
  swing1Motion=Free
  swing2LimitDegrees=45
  swing2Motion=Free
  twistLimitDegrees=45
  twistMotion=Free
 }
 {
  bone1Name=head
  bone2Name=neck_02
  swing1LimitDegrees=45
  swing1Motion=Free
  swing2LimitDegrees=45
  swing2Motion=Free
  twistLimitDegrees=45
  twistMotion=Free
 }
 {bone1Name=dyn_l_hair_b_01, bone2Name=head, swing1LimitDegrees=30, swing2LimitDegrees=15, twistLimitDegrees=0}
 {bone1Name=dyn_m_hair_01, bone2Name=head, swing1LimitDegrees=30, swing2LimitDegrees=15, twistLimitDegrees=0}
 {bone1Name=dyn_r_hair_b_01, bone2Name=head, swing1LimitDegrees=30, swing2LimitDegrees=15, twistLimitDegrees=0}
 {bone1Name=dyn_l_hair_b_02, bone2Name=dyn_l_hair_b_01, swing1LimitDegrees=30, swing2LimitDegrees=15, twistLimitDegrees=0}
 {bone1Name=dyn_l_hair_b_01, bone2Name=dyn_m_hair_01, swing1LimitDegrees=30, swing2LimitDegrees=15, twistLimitDegrees=0}
 {bone1Name=dyn_l_hair_b_02, bone2Name=dyn_m_hair_02, swing1LimitDegrees=30, swing2LimitDegrees=15, twistLimitDegrees=0}
 {bone1Name=dyn_m_hair_02, bone2Name=dyn_m_hair_01, swing1LimitDegrees=30, swing2LimitDegrees=15, twistLimitDegrees=0}
 {
  bone1Name=dyn_r_hair_b_01
  bone2Name=dyn_m_hair_01
  swing1LimitDegrees=45
  swing1Motion=Free
  swing2LimitDegrees=45
  swing2Motion=Free
  twistLimitDegrees=45
  twistMotion=Free
 }
 {
  bone1Name=dyn_r_hair_b_02
  bone2Name=dyn_m_hair_02
  swing1LimitDegrees=45
  swing1Motion=Free
  swing2LimitDegrees=45
  swing2Motion=Free
  twistLimitDegrees=45
  twistMotion=Free
 }
 {bone1Name=dyn_r_hair_b_02, bone2Name=dyn_r_hair_b_01, swing1LimitDegrees=30, swing2LimitDegrees=15, twistLimitDegrees=0}
]
