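# Physics asset for the hair on the FuzzyGlam female medium character's face accessory mesh, giving capsule bodies to hair and upper-body bones.
# Three two-bone hair chains (left, middle, right) hang from the head with narrow swing limits and zero-degree twist limits.
# An extra constraint ties the left root to the middle root.
# Upper spine, clavicle, neck and head capsules form the body shape, linked by angularly free constraints.
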PhysicsAsset /BRCosmetics/Characters/Player/Female/Medium/Bodies/F_MED_FuzzyGlam/Meshes/Parts/F_MED_FuzzyGlam_FaceAcc_Hair_Physics
bodies {
 clavicle_l [{shapeType=Capsule, length=18.75, radius=6.25, shapeName=clavicle_l_capsule, center=<4.9813 2.6958 -1.8597>, rotation=<rot 90 0.0005 0>}]
 clavicle_r [{shapeType=Capsule, length=18.75, radius=6.25, shapeName=clavicle_r_capsule, center=<-4.9813 -2.6958 1.8597>, rotation=<rot 90 -179.9995 0>}]
 dyn_l_hair_b_01 [{
   shapeType=Capsule
   length=7.125
   radius=4
   shapeName=dyn_l_hair_b_01_capsule
   center=<2.9383 -0.4 0.5>
   rotation=<rot 54.5641 -100.2689 -98.3969>
  }]
 dyn_l_hair_b_02 [{shapeType=Capsule, length=6.4954, radius=4.425, shapeName=dyn_l_hair_b_02_capsule, center=<3.6595 0.11 0>, rotation=<rot 90 -5 0>}]
 dyn_m_hair_01 [{shapeType=Capsule, length=6.625, radius=4, shapeName=dyn_m_hair_01_capsule, center=<1.6078 0 -0.12>, rotation=<rot 85.0001 180 180>}]
 dyn_m_hair_02 [{shapeType=Capsule, length=7.8704, radius=4.425, shapeName=dyn_m_hair_02_capsule, center=<2.7619 0 -1.7>, rotation=<rot 85 -180 -180>}]
 dyn_r_hair_b_01 [{shapeType=Capsule, length=3.8767, radius=4, shapeName=dyn_r_hair_b_01_capsule, center=<0.4849 0 0>, rotation=<rot 39.2586 103.0933 98.373>}]
 dyn_r_hair_b_02 [{
   shapeType=Capsule
   length=9.8597
   radius=4
   shapeName=dyn_r_hair_b_02_capsule
   center=<5.0618 3 2.3>
   rotation=<rot 55.0287 -89.6604 -101.4312>
  }]
 head [{shapeType=Capsule, length=5, radius=3, shapeName=head_capsule}]
 neck_01 [{shapeType=Capsule, length=5, radius=3, shapeName=neck_01_capsule}]
 neck_02 [{shapeType=Capsule, length=5, radius=3, shapeName=neck_02_capsule}]
 spine_05 [
  {
   shapeType=Capsule
   length=23.1304
   radius=13.125
   shapeName=spine_05_capsule
   center=<8.7543 7.3483 0.01>
   rotation=<rot 1.1499 -160.0004 180>
  }
  {shapeType=Capsule, length=32.625, radius=15, shapeName=spine_05_capsule, center=<-2.3348 7.0417 0>}
 ]
}
constraints [
 {
  bone1Name=neck_01
  bone2Name=spine_05
  swing1LimitDegrees=45
  swing1Motion=Free
  swing2LimitDegrees=45
  swing2Motion=Free
  twistLimitDegrees=45
  twistMotion=Free
 }
 {
  bone1Name=clavicle_l
  bone2Name=spine_05
  swing1LimitDegrees=45
  swing1Motion=Free
  swing2LimitDegrees=45
  swing2Motion=Free
  twistLimitDegrees=45
  twistMotion=Free
 }
 {
  bone1Name=clavicle_r
  bone2Name=spine_05
  swing1LimitDegrees=45
  swing1Motion=Free
  swing2LimitDegrees=45
  swing2Motion=Free
  twistLimitDegrees=45
  twistMotion=Free
 }
 {
  bone1Name=neck_02
  bone2Name=neck_01
  swing1LimitDegrees=45
  swing1Motion=Free
  swing2LimitDegrees=45
  swing2Motion=Free
  twistLimitDegrees=45
  twistMotion=Free
 }
 {
  bone1Name=head
  bone2Name=neck_02
  swing1LimitDegrees=45
  swing1Motion=Free
  swing2LimitDegrees=45
  swing2Motion=Free
  twistLimitDegrees=45
  twistMotion=Free
 }
 {bone1Name=dyn_l_hair_b_01, bone2Name=head, swing1LimitDegrees=30, swing2LimitDegrees=15, twistLimitDegrees=0}
 {bone1Name=dyn_m_hair_01, bone2Name=head, swing1LimitDegrees=10, swing2LimitDegrees=20, twistLimitDegrees=0}
 {bone1Name=dyn_r_hair_b_01, bone2Name=head, swing1LimitDegrees=15, swing2LimitDegrees=2, twistLimitDegrees=0}
 {bone1Name=dyn_l_hair_b_02, bone2Name=dyn_l_hair_b_01, swing1LimitDegrees=30, swing2LimitDegrees=0, twistLimitDegrees=0}
 {bone1Name=dyn_l_hair_b_01, bone2Name=dyn_m_hair_01, swing1LimitDegrees=10, swing2LimitDegrees=20, twistLimitDegrees=0}
 {bone1Name=dyn_m_hair_02, bone2Name=dyn_m_hair_01, swing1LimitDegrees=5, swing2LimitDegrees=30, twistLimitDegrees=0}
 {bone1Name=dyn_r_hair_b_02, bone2Name=dyn_r_hair_b_01, swing1LimitDegrees=30, swing2LimitDegrees=5, twistLimitDegrees=0}
]
