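# Control Rig for the FuzzyGlam female face accessory in standing pose, adjusting dynamic hair bones to follow body movement.
# Aim constraints point middle, left and right hair strands at Aim_Cons targets, which are absent from the listed hierarchy.
# Spherical pose readers on the right upper arm and head, relative to spine_05, drive smoothed corrective hair transforms.
# HeadsUp_Tagetting multiplies both reader outputs, so its default of zero leaves those correctives off.
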ControlRigBlueprint /BRCosmetics/Characters/Player/Female/Medium/Bodies/F_MED_FuzzyGlam/Meshes/Parts/F_MED_FuzzyGlam_FaceAcc_Standing_CtrlRig
graphs [
 RigVMFunctionLibrary {links [], nodes []}
 RigVMFunction_Sequence_1_ContainedGraph {nodes [
   n1 = RigVMFunctionEntryNode()
   RigVMFunction_Sequence = &RigVMFunction_Sequence(ExecuteContext=@n1.ExecuteContext)
   n2 = &RigVMFunction_Sequence(ExecuteContext=@RigVMFunction_Sequence.B)
   RigVMFunctionReturnNode(A=@RigVMFunction_Sequence.A, B=@n2.A, C=@n2.B)
  ]}
 RigVMModel {nodes [
   RigUnit_BeginExecution = &RigUnit_BeginExecution()
   RigVMFunction_Sequence = RigVMAggregateNode(ExecuteContext=@RigUnit_BeginExecution.ExecutePin)
   &RigUnit_AimConstraintLocalSpaceOffset(
    Child=(Type=Bone,Name="dyn_m_hair_01")
    Filter=(bX=True,bY=True,bZ=True)
    Parents=((Item=(Type=Bone,Name="Aim_Cons"),Weight=1))
    Weight=1
    WorldUp=(Target=(X=0,Y=0,Z=1),Kind=Location,Space=(Type=None,Name="None"))
    bMaintainOffset=true
    ExecutePin=@RigVMFunction_Sequence.A
   )
   > &RigUnit_AimConstraintLocalSpaceOffset(
    Child=(Type=Bone,Name="dyn_l_hair_b_01")
    Filter=(bX=false,bY=True,bZ=True)
    Parents=((Item=(Type=Bone,Name="Aim_Cons"),Weight=1))
    Weight=1
    WorldUp=(Target=(X=0,Y=0,Z=1),Kind=Location,Space=(Type=None,Name="None"))
    bMaintainOffset=true
   )
   > &RigUnit_AimConstraintLocalSpaceOffset(
    Child=(Type=Bone,Name="dyn_r_hair_b_01")
    Filter=(bX=false,bY=True,bZ=True)
    Parents=((Item=(Type=Bone,Name="Aim_Cons"),Weight=0.2),(Item=(Type=Bone,Name="Aim_Cons_r_arm"),Weight=0.8))
    Weight=1
    WorldUp=(Target=(X=0,Y=0,Z=1),Kind=Location,Space=(Type=None,Name="None"))
    bMaintainOffset=true
   )
   > n1 = &RigUnit_SphericalPoseReader(
    ActiveRegionScaleFactors=(PositiveWidth=1,NegativeWidth=1,PositiveHeight=1,NegativeHeight=1)
    ActiveRegionSize=0.1
    Debug=(bDrawDebug=true,bDraw2D=False,bDrawLocalAxes=False,DebugScale=25,DebugSegments=20,DebugThickness=0.25)
    DriverAxis=(X=-1,Y=0,Z=0)
    DriverItem=(Type=Bone,Name="upperarm_r")
    FalloffRegionScaleFactors=(PositiveWidth=1.5,NegativeWidth=1.8,PositiveHeight=1.8,NegativeHeight=1.5)
    FalloffSize=0.2
    FlipHeightScaling=False
    FlipWidthScaling=False
    OptionalParentItem=(Type=Bone,Name="spine_05")
    RotationOffset=(X=-100,Y=0,Z=40)
   )
   n2 = RigVMVariableNode() {variable=HeadsUp_Tagetting}
   Multiply = &RigVMFunction_MathDoubleMul(B=@n1.OutputParam, A=@n2.Value)
   AlphaInterp = &RigVMFunction_AlphaInterp(InterpSpeedDecreasing=10, InterpSpeedIncreasing=10, Scale=1, bClampResult=False, Value=@Multiply.Result)
   > &RigUnit_ModifyTransforms(
    ItemToModify="((Item=(Type=Bone,Name=\"dyn_r_hair_a_01\"),Transform=(Rotation=(X=-0.1713,Y=-0.2525,Z=0.3225,W=0.896),Translation=(X=-1,Y=0,Z=0),Scale3D=(X=1,Y=1,Z=1))),(Item=(Type=Bone,Name=\"dyn_r_hair_a_02\"),Transform=(Rotation=(X=-0.2588,Y=0,Z=0,W=0.9659),Translation=(X ...(truncated)"
    WeightMinimum=0
    Weight=@AlphaInterp.Result
   )
   n3 = &RigUnit_SphericalPoseReader(
    ActiveRegionScaleFactors=(PositiveWidth=1,NegativeWidth=1,PositiveHeight=1,NegativeHeight=1)
    ActiveRegionSize=0.1
    Debug=(bDrawDebug=true,bDraw2D=False,bDrawLocalAxes=False,DebugScale=25,DebugSegments=20,DebugThickness=0.25)
    DriverAxis=(X=1,Y=0,Z=0)
    DriverItem=(Type=Bone,Name="head")
    FalloffRegionScaleFactors=(PositiveWidth=1,NegativeWidth=1,PositiveHeight=1,NegativeHeight=1)
    FalloffSize=0.2
    FlipHeightScaling=False
    FlipWidthScaling=False
    OptionalParentItem=(Type=Bone,Name="spine_05")
    RotationOffset=(X=-65,Y=0,Z=35)
    ExecutePin=@RigVMFunction_Sequence.B
   )
   n4 = &RigVMFunction_AlphaInterp(
    InterpSpeedDecreasing=5
    InterpSpeedIncreasing=5
    Scale=1
    bClampResult=False
    Value=&RigVMFunction_MathDoubleMul(A=@n3.OutputParam, B=@n2.Value).Result
   )
   > &RigUnit_ModifyTransforms(
    ItemToModify="((Item=(Type=Bone,Name=\"dyn_r_hair_b_base\"),Transform=(Rotation=(X=0,Y=0,Z=0,W=1),Translation=(X=0,Y=0,Z=0),Scale3D=(X=1,Y=1,Z=1))),(Item=(Type=Bone,Name=\"dyn_l_hair_b_01\"),Transform=(Rotation=(X=0,Y=0,Z=0,W=1),Translation=(X=0,Y=0,Z=0),Scale3D=(X=1,Y=1,Z= ...(truncated)"
    WeightMinimum=0
    Weight=@n4.Result
   )
  ]}
]
hierarchy [
 root {loc=<0 0 0>, rot=<0 0 0 1>}
 pelvis {parent=root, loc=<0 -0.5301 98.0571>, rot=<0.0127 0.707 -0.0127 -0.707>}
 spine_01 {parent=pelvis, loc=<6.1208 -0.9328 0>, rot=<0 0 0.034 -0.9994>}
 spine_02 {parent=spine_01, loc=<6.5997 -0.2095 0>, rot=<0 0 0.0238 -0.9997>}
 spine_03 {parent=spine_02, loc=<6.4914 -0.1237 0>, rot=<0 0 0.0149 -0.9999>}
 spine_04 {parent=spine_03, loc=<6.9598 0.0794 0>, rot=<0 0 -0.0132 -0.9999>}
 spine_05 {parent=spine_04, loc=<6.9987 0 0>, rot=<0 0 -0.0015 -1>}
 neck_01 {parent=spine_05, loc=<11.3138 0.1529 0>, rot=<0 0 -0.2003 -0.9797>}
 neck_02 {parent=neck_01, loc=<4.5627 0 0>, rot=<0 0 0.0504 -0.9987>}
 head {parent=neck_02, loc=<5.3914 0 0>, rot=<0 0 0.115 -0.9934>}
 dyn_r_feather_01 {parent=head, loc=<6.0988 -0.9938 9.2719>, rot=<-0.002 -0.0017 0.1086 -0.9941>}
 dyn_r_feather_02 {parent=dyn_r_feather_01, loc=<5.7325 0 0>, rot=<-0.0006 -0.185 -0.0063 -0.9827>}
 dyn_l_feather_01 {parent=head, loc=<6.0454 -0.9146 -8.8812>, rot=<0.9941 -0.1086 -0.0017 -0.002>}
 dyn_l_feather_02 {parent=dyn_l_feather_01, loc=<5.7325 0 0>, rot=<-0.0006 -0.185 -0.0063 -0.9827>}
 dyn_f_hair {parent=head, loc=<8.1837 8.8982 0.7253>, rot=<0.0957 0.7006 0.7006 0.0957>}
 dyn_r_hair_01 {parent=head, loc=<4.0126 2.4153 6.1565>, rot=<0.0764 0.3814 0.9213 -0.0011>}
 dyn_l_hair_01 {parent=head, loc=<3.7424 2.5611 -5.6943>, rot=<0.1122 -0.9865 0.0125 -0.1182>}
 dyn_l_hair_b_base {parent=head, loc=<6.4875 -6.5082 -4.9445>, rot=<0.0661 -0.8917 0.4425 -0.0682>}
 dyn_l_hair_b_01 {parent=dyn_l_hair_b_base, loc=<0.7613 -0.0285 0.4362>, rot=<0 0 0 1>}
 dyn_l_hair_b_02 {parent=dyn_l_hair_b_01, loc=<7.9067 -0.1331 0.0968>, rot=<0 0 0 1>}
 dyn_l_hair_b_03 {parent=dyn_l_hair_b_02, loc=<6.5143 0.1615 -0.533>, rot=<0 0 0 1>}
 dyn_r_hair_b_base {parent=head, loc=<5.8297 -7.2014 6.5002>, rot=<0.0242 -0.2789 0.96 -0.0089>}
 dyn_r_hair_b_01 {parent=dyn_r_hair_b_base, loc=<0.0285 -0.0124 0.1778>, rot=<0 0 0 1>}
 dyn_r_hair_b_02 {parent=dyn_r_hair_b_01, loc=<8.0142 0.0029 -0.041>, rot=<0 0 0 1>}
 dyn_r_hair_b_03 {parent=dyn_r_hair_b_02, loc=<5.9372 0.0096 -0.1368>, rot=<0 0 0 1>}
 dyn_r_hair_a_base1 {parent=head, loc=<1.5342 -2.1947 7>, rot=<0 0 0.9989 -0.0479>}
 dyn_r_hair_a_01 {parent=dyn_r_hair_a_base1, loc=<0 0 0>, rot=<0 0 0 1>}
 dyn_r_hair_a_02 {parent=dyn_r_hair_a_01, loc=<6.1574 0 0>, rot=<0 0 0 1>}
 dyn_r_hair_a_03 {parent=dyn_r_hair_a_02, loc=<4.0556 0 0>, rot=<0 0 0 1>}
 dyn_l_hair_a_base {parent=head, loc=<0.818 -1.3 -7>, rot=<0.0479 -0.9989 0 0>}
 dyn_l_hair_a_01 {parent=dyn_l_hair_a_base, loc=<0 0 0>, rot=<0 0 0 1>}
 dyn_l_hair_a_02 {parent=dyn_l_hair_a_01, loc=<6.1574 0 0>, rot=<0 0 0 1>}
 dyn_l_hair_a_03 {parent=dyn_l_hair_a_02, loc=<3.5403 0 0>, rot=<0 0 0 1>}
 dyn_m_hair_base {parent=head, loc=<4.8272 -9.5671 0>, rot=<0.0623 -0.7044 0.7044 -0.0623>}
 dyn_m_hair_01 {parent=dyn_m_hair_base, loc=<0 0 0>, rot=<0 0 0 1>}
 dyn_m_hair_02 {parent=dyn_m_hair_01, loc=<7.1428 0 0>, rot=<0 0 0 1>}
 dyn_m_hair_03 {parent=dyn_m_hair_02, loc=<4.5272 0 0>, rot=<0 0 0 1>}
 clavicle_l {parent=spine_05, loc=<5.4831 -0.1053 -3.2946>, rot=<0.0465 -0.7289 -0.009 -0.6829>}
 upperarm_l {parent=clavicle_l, loc=<9.8658 -1.2136 -0.4105>, rot=<0.0164 -0.3245 -0.0269 -0.9454>}
 lowerarm_l {parent=upperarm_l, loc=<26.465 -0.8345 0.259>, rot=<0.0036 0.0161 -0.2298 -0.9731>}
 upperarm_twist_01_l {parent=upperarm_l, loc=<7.2135 -0.1134 0>, rot=<0 0 0 1>}
 clavicle_r {parent=spine_05, loc=<5.4831 -0.1053 3.2946>, rot=<0.7289 0.0465 -0.6829 0.009>}
 upperarm_r {parent=clavicle_r, loc=<-9.8658 1.2136 0.4105>, rot=<0.0164 -0.3245 -0.0269 -0.9454>}
 lowerarm_r {parent=upperarm_r, loc=<-26.465 0.8345 -0.259>, rot=<0.0036 0.0161 -0.2298 -0.9731>}
 upperarm_twist_01_r {parent=upperarm_r, loc=<-7.2135 0.1134 0>, rot=<0 0 0 -1>}
 attach_cape {parent=spine_05, loc=<9.017 -11.714 0>, rot=<0.0616 -0.7044 -0.0616 -0.7044>}
]
variables [HeadsUp_Tagetting {type=double, default=0}]
