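# Animation Blueprint for the FuzzyGlam female face accessory, driving secondary motion on hair and feather bones.
# Crouched (misleadingly cached as "Not CrocuhDYN") and standing branches combine Control Rigs, additive bone offsets, AnimDynamics and, when standing, RigidBody with a front-end physics variant.
# Emotes switch to separate RigidBody physics before a final GunHair Control Rig.
# The event graph derives gravity settings, head angles and targeting, crouch and emote values, several feeding the rigs.
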AnimBlueprint /BRCosmetics/Characters/Player/Female/Medium/Bodies/F_MED_FuzzyGlam/Meshes/Parts/F_MED_FuzzyGlam_FaceAcc_AnimBP
target_skeleton=&/F_MED_FuzzyGlam_FaceAcc_Skeleton
graphs [
 AnimGraph {nodes [
   n1 = RigidBody!(
    ComponentPose=Local To Component(LocalPose=Use cached pose 'IntialPose'())
    Alpha
    OverrideWorldGravity:Vector="0,0,0"
    ExternalForce
    SimSpaceSettings:Sim Space Settings Structure=(WorldAlpha=0,VelocityScaleZ=1,DampingAlpha=1,MaxLinearVelocity=10000,MaxAngularVelocity=10000,MaxLinearAcceleration=10000,MaxAngularAcceleration=10000,ExternalLinearDragV=(X=0,Y=0,Z=0),ExternalLinearVelocity=(X=0,Y=0,Z=0),ExternalAngularVelocity=(X=0,Y=0,Z=0))
   ) -> (Pose) {
    bOverrideWorldGravity=true
    overridePhysicsAsset=&/F_MED_FuzzyGlam_FaceAcc_Hair_Emote_Physics
    simulationSpace=BaseBoneSpace
    alphaScaleBiasClamp {interpSpeedDecreasing=5, interpSpeedIncreasing=5}
    baseBoneRef {boneName=head}
   }
   n2 = "Transform (Modify) Bone - Bone: dyn_r_hair_b_01"(ComponentPose=@n1, Alpha=1, Translation="0,0,0", Rotation="14.651,10,2.662", Scale="1,1,1") {
    rotationMode=BMM_Additive
    rotationSpace=BCS_BoneSpace
    scaleMode=BMM_Additive
    scaleSpace=BCS_BoneSpace
    translationMode=BMM_Additive
    translationSpace=BCS_BoneSpace
    alphaScaleBiasClamp {bInterpResult=true}
    boneToModify {boneName=dyn_r_hair_b_01}
    rotation=<rot 14.651 10 2.662>
   }
   n3 = "Transform (Modify) Bone - Bone: dyn_m_hair_01"(ComponentPose=@n2, Alpha=1, Translation="0,0,0", Rotation="10,0,0", Scale="1,1,1") {
    rotationMode=BMM_Additive
    rotationSpace=BCS_BoneSpace
    scaleMode=BMM_Additive
    scaleSpace=BCS_BoneSpace
    translationMode=BMM_Additive
    translationSpace=BCS_BoneSpace
    alphaScaleBiasClamp {bInterpResult=true}
    boneToModify {boneName=dyn_m_hair_01}
    rotation {pitch=10}
   }
   n4 = "Transform (Modify) Bone - Bone: dyn_l_hair_b_01"(ComponentPose=@n3, Alpha=1, Translation="0,0,0", Rotation="10,0,0", Scale="1,1,1") {
    rotationMode=BMM_Additive
    rotationSpace=BCS_BoneSpace
    scaleMode=BMM_Additive
    scaleSpace=BCS_BoneSpace
    translationMode=BMM_Additive
    translationSpace=BCS_BoneSpace
    alphaScaleBiasClamp {bInterpResult=true}
    boneToModify {boneName=dyn_l_hair_b_01}
    rotation {pitch=10}
   }
   n5 = Control Rig!(
    Source=Blend Poses by bool(
     BlendPose_0=Component To Local(ComponentPose=@n4)
     BlendPose_1=Use cached pose 'DYN'()
     BlendTime_0=1
     BlendTime_1=1
     bActiveValue=Get bIsPlayingEmote()
    )
    Alpha
    isADS:double=Get IsTragetting() -> (IsTragetting:double)
    isADS_Head_Crouch:double=Get CrouchHeadUpTargetting() -> (CrouchHeadUpTargetting:double)
    Emote_HeadRot:double=Get Emote_HeadRot() -> (Emote_HeadRot:double)
    RBANEnabledFloat:double=Get RBEnabledFloat() -> (RBEnabledFloat:double)
   ) -> (Pose) {controlRigAssetReference {blueprintRigClass=&/F_MED_FuzzyGlam_FaceAcc_CtrlRig_GunHair.F_MED_FuzzyGlam_FaceAcc_CtrlRig_GunHair_C}}
   Output Pose(Result=@n5) {layerGroup=None}
   n6 = Control Rig!(
    Source=Use cached pose 'IntialPose'()
    Alpha
    CrouchHeadUpWalk:double=Get CrouchHeadUpWalk() -> (CrouchHeadUpWalk:double)
    CrouchWalk:bool=false
    InPlaceMove:double=Get InPlaceMove() -> (InPlaceMove:double)
   ) -> (Pose) {controlRigAssetReference {blueprintRigClass=&/F_MED_FuzzyGlam_FaceAcc_CtrlRig.F_MED_FuzzyGlam_FaceAcc_CtrlRig_C}}
   n7 = "Transform (Modify) Bone - Bone: dyn_r_hair_b_01"(ComponentPose=Local To Component(LocalPose=@n6), Alpha=1, Translation="0,0,0", Rotation="-0.000,10.000,-0.000", Scale="1,1,1") {
    rotationMode=BMM_Additive
    rotationSpace=BCS_BoneSpace
    scaleMode=BMM_Additive
    scaleSpace=BCS_BoneSpace
    translationMode=BMM_Additive
    translationSpace=BCS_BoneSpace
    alphaScaleBiasClamp {bInterpResult=true}
    boneToModify {boneName=dyn_r_hair_b_01}
    rotation=<rot -0.0001 9.9999 0>
   }
   n8 = "Transform (Modify) Bone - Bone: dyn_m_hair_01"(ComponentPose=@n7, Alpha=1, Translation="0,0,0", Rotation="0,10,0", Scale="1,1,1") {
    rotationMode=BMM_Additive
    rotationSpace=BCS_BoneSpace
    scaleMode=BMM_Additive
    scaleSpace=BCS_BoneSpace
    translationMode=BMM_Additive
    translationSpace=BCS_BoneSpace
    alphaScaleBiasClamp {bInterpResult=true}
    boneToModify {boneName=dyn_m_hair_01}
    rotation {yaw=10}
   }
   n9 = "Transform (Modify) Bone - Bone: dyn_m_hair_01"(ComponentPose=@n8, Alpha=1, Translation="0,0,0", Rotation="0,10,0", Scale="1,1,1") {
    rotationMode=BMM_Additive
    rotationSpace=BCS_BoneSpace
    scaleMode=BMM_Additive
    scaleSpace=BCS_BoneSpace
    translationMode=BMM_Additive
    translationSpace=BCS_BoneSpace
    alphaScaleBiasClamp {bInterpResult=true}
    boneToModify {boneName=dyn_m_hair_01}
    rotation {yaw=10}
   }
   n10 = Anim Dynamics - Bone: dyn_l_hair_b_01!(ComponentPose=@n9, Alpha=1, GravityOverride:Vector="0,0,-980", ExternalForce="0,0,0") -> (Pose) {
    angularDampingOverride=1
    angularSpringConstant=10000
    bAngularSpring=true
    bOverrideAngularDamping=true
    bUseGravityOverride=true
    gravityScale=0.35
    simulationSpace=RootRelative
    boundBone {boneName=dyn_l_hair_b_01}
    componentAppliedLinearAccClamp=<5000 5000 5000>
    componentLinearAccScale=<1 1 1>
    componentLinearVelScale=<1 1 1>
    gravityOverride {z=-980}
    physicsBodyDefinitions [{
      boundBone {boneName=dyn_l_hair_b_01}
      boxExtents=<20 30 30>
      constraintSetup {angularLimitsMax=<0 5 5>, angularLimitsMin=<0 -30 -5>, angularTarget=<1 0 0>}
      localJointOffset=<3 0 0>
     }]
   }
   n11 = Anim Dynamics - Bone: dyn_l_hair_b_02!(ComponentPose=@n10, Alpha=1, GravityOverride:Vector="0,0,-980", ExternalForce="0,0,0") -> (Pose) {
    angularDampingOverride=1
    angularSpringConstant=10000
    bAngularSpring=true
    bOverrideAngularDamping=true
    bUseGravityOverride=true
    gravityScale=0.35
    simulationSpace=RootRelative
    boundBone {boneName=dyn_l_hair_b_02}
    componentAppliedLinearAccClamp=<5000 5000 5000>
    componentLinearAccScale=<1 1 1>
    componentLinearVelScale=<1 1 1>
    gravityOverride {z=-980}
    physicsBodyDefinitions [{
      boundBone {boneName=dyn_l_hair_b_02}
      boxExtents=<20 30 30>
      constraintSetup {angularLimitsMax=<0 5 5>, angularLimitsMin=<0 -30 -5>, angularTarget=<1 0 0>}
      localJointOffset=<3 0 0>
     }]
   }
   n12 = Anim Dynamics - Bone: dyn_m_hair_01!(ComponentPose=@n11, Alpha=1, GravityOverride:Vector="0,0,-980", ExternalForce="0,0,0") -> (Pose) {
    angularDampingOverride=1
    angularSpringConstant=10000
    bAngularSpring=true
    bOverrideAngularDamping=true
    bUseGravityOverride=true
    gravityScale=0.35
    simulationSpace=RootRelative
    boundBone {boneName=dyn_m_hair_01}
    componentAppliedLinearAccClamp=<5000 5000 5000>
    componentLinearAccScale=<1 1 1>
    componentLinearVelScale=<1 1 1>
    gravityOverride {z=-980}
    physicsBodyDefinitions [{
      boundBone {boneName=dyn_m_hair_01}
      boxExtents=<20 30 30>
      constraintSetup {angularLimitsMax=<0 5 5>, angularLimitsMin=<0 -30 -5>, angularTarget=<1 0 0>}
      localJointOffset=<3 0 0>
     }]
   }
   n13 = Anim Dynamics - Bone: dyn_m_hair_02!(ComponentPose=@n12, Alpha=1, GravityOverride:Vector="0,0,-980", ExternalForce="0,0,0") -> (Pose) {
    angularDampingOverride=1
    angularSpringConstant=10000
    bAngularSpring=true
    bOverrideAngularDamping=true
    bUseGravityOverride=true
    gravityScale=0.35
    simulationSpace=RootRelative
    boundBone {boneName=dyn_m_hair_02}
    componentAppliedLinearAccClamp=<5000 5000 5000>
    componentLinearAccScale=<1 1 1>
    componentLinearVelScale=<1 1 1>
    gravityOverride {z=-980}
    physicsBodyDefinitions [{
      boundBone {boneName=dyn_m_hair_02}
      boxExtents=<20 30 30>
      constraintSetup {angularLimitsMax=<0 5 5>, angularLimitsMin=<0 -30 -5>, angularTarget=<1 0 0>}
      localJointOffset=<3 0 0>
     }]
   }
   n14 = "Transform (Modify) Bone - Bone: dyn_l_hair_a_01"(ComponentPose=@n13, Alpha=1, Translation="0.052,-0.172,-0.000", Rotation="0.000,-85.000,-0.000", Scale="1,1,1") {
    rotationMode=BMM_Additive
    rotationSpace=BCS_BoneSpace
    scaleMode=BMM_Additive
    scaleSpace=BCS_BoneSpace
    translationMode=BMM_Additive
    translationSpace=BCS_BoneSpace
    alphaScaleBiasClamp {bInterpResult=true, interpSpeedDecreasing=5, interpSpeedIncreasing=5}
    boneToModify {boneName=dyn_l_hair_a_01}
    rotation=<rot 0 -84.9999 0>
    translation=<0.052 -0.172>
   }
   n15 = "Transform (Modify) Bone - Bone: dyn_l_hair_a_02"(ComponentPose=@n14, Alpha=1, Translation="0.548,-0.000,0.000", Rotation="0,0,0", Scale="1,1,1") {
    rotationMode=BMM_Additive
    rotationSpace=BCS_BoneSpace
    scaleMode=BMM_Additive
    scaleSpace=BCS_BoneSpace
    translationMode=BMM_Additive
    translationSpace=BCS_BoneSpace
    alphaScaleBiasClamp {bInterpResult=true, interpSpeedDecreasing=5, interpSpeedIncreasing=5}
    boneToModify {boneName=dyn_l_hair_a_02}
    translation {x=0.548}
   }
   n16 = "Transform (Modify) Bone - Bone: dyn_r_hair_a_01"(ComponentPose=@n15, Alpha=1, Translation="0,0,0", Rotation="0.000,60.000,-0.000", Scale="1,1,1") {
    rotationMode=BMM_Additive
    rotationSpace=BCS_BoneSpace
    scaleMode=BMM_Additive
    scaleSpace=BCS_BoneSpace
    translationMode=BMM_Additive
    translationSpace=BCS_BoneSpace
    alphaScaleBiasClamp {bInterpResult=true, interpSpeedDecreasing=5, interpSpeedIncreasing=5}
    boneToModify {boneName=dyn_r_hair_a_01}
    rotation=<rot 0 59.9999 0>
   }
   n17 = "Transform (Modify) Bone - Bone: dyn_m_hair_02"(ComponentPose=@n16, Alpha=1, Translation="0,0,0", Rotation="0,0,0", Scale="1,1,1") {
    rotationMode=BMM_Additive
    rotationSpace=BCS_BoneSpace
    scaleMode=BMM_Additive
    scaleSpace=BCS_BoneSpace
    translationMode=BMM_Additive
    translationSpace=BCS_BoneSpace
    alphaScaleBiasClamp {bInterpResult=true, interpSpeedDecreasing=5, interpSpeedIncreasing=5}
    boneToModify {boneName=dyn_m_hair_02}
   }
   n18 = "Transform (Modify) Bone - Bone: dyn_l_hair_a_01"(ComponentPose=@n17, Alpha=1, Translation="0.176,-0.967,-0.053", Rotation="-11.669,-64.382,-37.063", Scale="1,1,1") {
    rotationMode=BMM_Additive
    rotationSpace=BCS_BoneSpace
    translationMode=BMM_Additive
    translationSpace=BCS_BoneSpace
    alphaScaleBiasClamp {bInterpResult=true, interpSpeedDecreasing=5, interpSpeedIncreasing=5}
    boneToModify {boneName=dyn_l_hair_a_01}
    rotation=<rot -11.669 -64.382 -37.063>
    translation=<0.176 -0.967 -0.053>
   }
   n19 = "Transform (Modify) Bone - Bone: dyn_l_hair_a_02"(ComponentPose=@n18, Alpha=1, Translation="0,0,0", Rotation="4.994,-5.001,-0.125", Scale="1,1,1") {
    rotationMode=BMM_Additive
    rotationSpace=BCS_BoneSpace
    translationMode=BMM_Additive
    translationSpace=BCS_BoneSpace
    alphaScaleBiasClamp {bInterpResult=true, interpSpeedDecreasing=5, interpSpeedIncreasing=5}
    boneToModify {boneName=dyn_l_hair_a_02}
    rotation=<rot 4.994 -5.001 -0.125>
   }
   n20 = "Transform (Modify) Bone - Bone: dyn_l_hair_b_01"(ComponentPose=@n19, Alpha=1, Translation="-0.000,0.000,-0.716", Rotation="4.922,19.980,0.680", Scale="1,1,1") {
    rotationMode=BMM_Additive
    rotationSpace=BCS_BoneSpace
    translationMode=BMM_Additive
    translationSpace=BCS_BoneSpace
    alphaScaleBiasClamp {bInterpResult=true, interpSpeedDecreasing=5, interpSpeedIncreasing=5}
    boneToModify {boneName=dyn_l_hair_b_01}
    rotation=<rot 4.922 19.98 0.68>
    translation {z=-0.716}
   }
   n21 = "Transform (Modify) Bone - Bone: dyn_l_hair_b_02"(ComponentPose=@n20, Alpha=1, Translation="-4.946,0.000,-1.424", Rotation="0,0,0", Scale="1,1,1") {
    rotationMode=BMM_Additive
    rotationSpace=BCS_BoneSpace
    translationMode=BMM_Additive
    translationSpace=BCS_BoneSpace
    alphaScaleBiasClamp {bInterpResult=true, interpSpeedDecreasing=5, interpSpeedIncreasing=5}
    boneToModify {boneName=dyn_l_hair_b_02}
    translation {x=-4.946, z=-1.424}
   }
   n22 = "Transform (Modify) Bone - Bone: dyn_m_hair_01"(ComponentPose=@n21, Alpha=1, Translation="0,0,0", Rotation="-0.000,14.988,0.003", Scale="1,1,1") {
    rotationMode=BMM_Additive
    rotationSpace=BCS_BoneSpace
    scaleMode=BMM_Additive
    scaleSpace=BCS_BoneSpace
    translationMode=BMM_Additive
    translationSpace=BCS_BoneSpace
    alphaScaleBiasClamp {bInterpResult=true, interpSpeedDecreasing=5, interpSpeedIncreasing=5}
    boneToModify {boneName=dyn_m_hair_01}
    rotation=<rot 0 14.988 0.003>
   }
   n23 = "Transform (Modify) Bone - Bone: dyn_m_hair_02"(ComponentPose=@n22, Alpha=1, Translation="-3.776,0.095,-0.215", Rotation="14.877,0.408,2.421", Scale="1,1,1") {
    rotationMode=BMM_Additive
    rotationSpace=BCS_BoneSpace
    scaleMode=BMM_Additive
    scaleSpace=BCS_BoneSpace
    translationMode=BMM_Additive
    translationSpace=BCS_BoneSpace
    alphaScaleBiasClamp {bInterpResult=true, interpSpeedDecreasing=5, interpSpeedIncreasing=5}
    boneToModify {boneName=dyn_m_hair_02}
    rotation=<rot 14.877 0.408 2.421>
    translation=<-3.776 0.095 -0.215>
   }
   n24 = "Transform (Modify) Bone - Bone: dyn_r_hair_b_01"(ComponentPose=@n23, Alpha=1, Translation="0,0,0", Rotation="-3.038,49.085,9.067", Scale="1,1,1") {
    rotationMode=BMM_Additive
    rotationSpace=BCS_BoneSpace
    scaleMode=BMM_Additive
    scaleSpace=BCS_BoneSpace
    translationMode=BMM_Additive
    translationSpace=BCS_BoneSpace
    alphaScaleBiasClamp {bInterpResult=true, interpSpeedDecreasing=5, interpSpeedIncreasing=5}
    boneToModify {boneName=dyn_r_hair_b_01}
    rotation=<rot -3.0381 49.0849 9.0669>
   }
   n25 = "Transform (Modify) Bone - Bone: dyn_r_hair_b_01"(ComponentPose=@n24, Alpha=1, Translation="0,0,0", Rotation="-0.000,15.000,-0.000", Scale="1,1,1") {
    rotationMode=BMM_Additive
    rotationSpace=BCS_BoneSpace
    scaleMode=BMM_Additive
    scaleSpace=BCS_BoneSpace
    translationMode=BMM_Additive
    translationSpace=BCS_BoneSpace
    alphaScaleBiasClamp {interpSpeedDecreasing=1, interpSpeedIncreasing=1}
    boneToModify {boneName=dyn_r_hair_b_01}
    rotation=<rot 0 15 0>
   }
   Save cached pose 'Not CrocuhDYN'(Pose=Component To Local(ComponentPose=@n25)) {cacheName=Not CrocuhDYN}
   n26 = Layered blend per bone(
    BasePose=Copy Pose From Mesh() {bUseAttachedParent=true}
    BlendPoses_0=Copy Pose From Mesh(SourceMeshComponent=Get Body() -> (Body:Skeletal Mesh Component Budgeted Object Reference)) {bUseAttachedParent=true}
   ) {layerSetup [{branchFilters [
       {blendDepth=2, boneName=clavicle_r}
       {blendDepth=2, boneName=clavicle_l}
       {blendDepth=2, boneName=upperarm_l}
       {blendDepth=2, boneName=upperarm_r}
      ]}]}
   Save cached pose 'IntialPose'(Pose=@n26) {cacheName=IntialPose}
   n27 = Use cached pose 'Not CrocuhDYN'()
   n28 = Get bIsCrouching()
   n29 = Control Rig!(
    Source=Copy Pose From Mesh() {bUseAttachedParent=true}
    Alpha
    HeadsUp_Tagetting:double=Get HeadsUp_Tagetting() -> (HeadsUp_Tagetting:double)
   ) -> (Pose) {controlRigAssetReference {blueprintRigClass=&/F_MED_FuzzyGlam_FaceAcc_Standing_CtrlRig.F_MED_FuzzyGlam_FaceAcc_Standing_CtrlRig_C}}
   n30 = Get bIsInFrontEnd()
   n31 = RigidBody!(
    ComponentPose=Local To Component(LocalPose=@n29)
    bAlphaBoolEnabled:bool=NOT Boolean(A=@n30)
    OverrideWorldGravity="0,0,0"
    ExternalForce
    SimSpaceSettings=(WorldAlpha=0,VelocityScaleZ=1,DampingAlpha=1,MaxLinearVelocity=10000,MaxAngularVelocity=10000,MaxLinearAcceleration=10000,MaxAngularAcceleration=10000,ExternalLinearDragV=(X=0,Y=0,Z=0),ExternalLinearVelocity=(X=0,Y=0,Z=0),ExternalAngularVelocity=(X=0,Y=0,Z=0))
   ) -> (Pose) {
    alphaInputType=Bool
    bOverrideWorldGravity=true
    overridePhysicsAsset=&/F_MED_FuzzyGlam_FaceAcc_Hair_Physics
    simulationSpace=BaseBoneSpace
    alphaScaleBiasClamp {interpSpeedDecreasing=5, interpSpeedIncreasing=5}
    baseBoneRef {boneName=head}
   }
   n32 = RigidBody(
    ComponentPose=@n31
    bAlphaBoolEnabled=@n30
    OverrideWorldGravity="0,0,0"
    SimSpaceSettings=(WorldAlpha=0,VelocityScaleZ=1,DampingAlpha=1,MaxLinearVelocity=10000,MaxAngularVelocity=10000,MaxLinearAcceleration=10000,MaxAngularAcceleration=10000,ExternalLinearDragV=(X=0,Y=0,Z=0),ExternalLinearVelocity=(X=0,Y=0,Z=0),ExternalAngularVelocity=(X=0,Y=0,Z=0))
   ) {
    alphaInputType=Bool
    bOverrideWorldGravity=true
    overridePhysicsAsset=&/F_MED_FuzzyGlam_FaceAcc_Hair_FE_Physics
    simulationSpace=BaseBoneSpace
    alphaScaleBiasClamp {interpSpeedDecreasing=5, interpSpeedIncreasing=5}
    baseBoneRef {boneName=head}
   }
   n33 = "Transform (Modify) Bone - Bone: dyn_l_hair_a_01"(ComponentPose=@n32, Alpha=1, Translation="0.000,-2.657,0.000", Rotation="0,0,0", Scale="1,1,1") {
    rotationMode=BMM_Additive
    rotationSpace=BCS_BoneSpace
    scaleMode=BMM_Additive
    scaleSpace=BCS_BoneSpace
    translationMode=BMM_Additive
    translationSpace=BCS_BoneSpace
    alphaScaleBiasClamp {bInterpResult=true}
    boneToModify {boneName=dyn_l_hair_a_01}
    translation {y=-2.657}
   }
   n34 = "Transform (Modify) Bone - Bone: dyn_l_hair_a_01"(ComponentPose=@n33, Alpha=1, Translation="-1.065,-1.272,-0.010", Rotation="10,0,0", Scale="1,1,1") {
    rotationMode=BMM_Additive
    rotationSpace=BCS_BoneSpace
    scaleMode=BMM_Additive
    scaleSpace=BCS_BoneSpace
    translationMode=BMM_Additive
    translationSpace=BCS_BoneSpace
    alphaScaleBiasClamp {bInterpResult=true}
    boneToModify {boneName=dyn_l_hair_a_01}
    rotation {pitch=10}
    translation=<-1.065 -1.2722 -0.01>
   }
   n35 = "Transform (Modify) Bone - Bone: dyn_r_hair_a_01"(ComponentPose=@n34, Alpha=1, Translation="0,0,0", Rotation="10,0,0", Scale="1,1,1") {
    rotationMode=BMM_Additive
    rotationSpace=BCS_BoneSpace
    scaleMode=BMM_Additive
    scaleSpace=BCS_BoneSpace
    translationMode=BMM_Additive
    translationSpace=BCS_BoneSpace
    alphaScaleBiasClamp {bInterpResult=true}
    boneToModify {boneName=dyn_r_hair_a_01}
    rotation {pitch=10}
   }
   n36 = "Transform (Modify) Bone - Bone: dyn_l_hair_a_01"(ComponentPose=@n35, Alpha=1, Translation="-0.405,-1.707,0.000", Rotation="-0.000,-45.848,-0.000", Scale="1,1,1") {
    rotationMode=BMM_Additive
    rotationSpace=BCS_BoneSpace
    scaleMode=BMM_Additive
    scaleSpace=BCS_BoneSpace
    translationMode=BMM_Additive
    translationSpace=BCS_BoneSpace
    alphaScaleBiasClamp {bInterpResult=true}
    boneToModify {boneName=dyn_l_hair_a_01}
    rotation=<rot 0 -45.848 0>
    translation=<-0.405 -1.707>
   }
   n37 = "Transform (Modify) Bone - Bone: dyn_l_hair_b_02"(ComponentPose=@n36, Alpha=1, Translation="-4.941,-0.596,0.851", Rotation="-4.895,35.044,-1.023", Scale="1,1,1") {
    rotationMode=BMM_Additive
    rotationSpace=BCS_BoneSpace
    scaleMode=BMM_Additive
    scaleSpace=BCS_BoneSpace
    translationMode=BMM_Additive
    translationSpace=BCS_BoneSpace
    alphaScaleBiasClamp {bInterpResult=true}
    boneToModify {boneName=dyn_l_hair_b_02}
    rotation=<rot -4.895 35.044 -1.023>
    translation=<-4.941 -0.596 0.851>
   }
   n38 = "Transform (Modify) Bone - Bone: dyn_m_hair_02"(ComponentPose=@n37, Alpha=1, Translation="-2.472,1.988,0.002", Rotation="-1.023,-32.243,0.801", Scale="1,1,1") {
    rotationMode=BMM_Additive
    rotationSpace=BCS_BoneSpace
    scaleMode=BMM_Additive
    scaleSpace=BCS_BoneSpace
    translationMode=BMM_Additive
    translationSpace=BCS_BoneSpace
    alphaScaleBiasClamp {bInterpResult=true}
    boneToModify {boneName=dyn_m_hair_02}
    rotation=<rot -1.0231 -32.243 0.801>
    translation=<-2.472 1.988 0.002>
   }
   n39 = "Transform (Modify) Bone - Bone: dyn_r_hair_a_01"(ComponentPose=@n38, Alpha=1, Translation="-2.102,0.806,-0.254", Rotation="-6.067,44.214,17.901", Scale="1,1,1") {
    rotationMode=BMM_Additive
    rotationSpace=BCS_BoneSpace
    scaleMode=BMM_Additive
    scaleSpace=BCS_BoneSpace
    translationMode=BMM_Additive
    translationSpace=BCS_BoneSpace
    alphaScaleBiasClamp {bInterpResult=true}
    boneToModify {boneName=dyn_r_hair_a_01}
    rotation=<rot -6.067 44.2139 17.901>
    translation=<-2.102 0.806 -0.254>
   }
   n40 = "Transform (Modify) Bone - Bone: dyn_r_hair_a_02"(ComponentPose=@n39, Alpha=1, Translation="2.225,2.671,1.506", Rotation="28.948,22.253,4.000", Scale="1,1,1") {
    rotationMode=BMM_Additive
    rotationSpace=BCS_BoneSpace
    scaleMode=BMM_Additive
    scaleSpace=BCS_BoneSpace
    translationMode=BMM_Additive
    translationSpace=BCS_BoneSpace
    alphaScaleBiasClamp {bInterpResult=true}
    boneToModify {boneName=dyn_r_hair_a_02}
    rotation=<rot 28.948 22.253 4>
    translation=<2.225 2.671 1.506>
   }
   n41 = "Transform (Modify) Bone - Bone: dyn_r_hair_a_03"(ComponentPose=@n40, Alpha=1, Translation="-8.877,0.002,0.308", Rotation="0,0,0", Scale="1,1,1") {
    rotationMode=BMM_Additive
    rotationSpace=BCS_BoneSpace
    scaleMode=BMM_Additive
    scaleSpace=BCS_BoneSpace
    translationMode=BMM_Additive
    translationSpace=BCS_BoneSpace
    alphaScaleBiasClamp {bInterpResult=true}
    boneToModify {boneName=dyn_r_hair_a_03}
    translation=<-8.877 0.002 0.308>
   }
   n42 = "Transform (Modify) Bone - Bone: dyn_r_hair_b_01"(ComponentPose=@n41, Alpha=1, Translation="0,0,0", Rotation="6.705,55.031,0.555", Scale="1,1,1") {
    rotationMode=BMM_Additive
    rotationSpace=BCS_BoneSpace
    scaleMode=BMM_Additive
    scaleSpace=BCS_BoneSpace
    translationMode=BMM_Additive
    translationSpace=BCS_BoneSpace
    alphaScaleBiasClamp {bInterpResult=true}
    boneToModify {boneName=dyn_r_hair_b_01}
    rotation=<rot 6.705 55.031 0.555>
   }
   n43 = "Transform (Modify) Bone - Bone: dyn_r_hair_b_02"(ComponentPose=@n42, Alpha=1, Translation="-0.958,0.156,-0.731", Rotation="37.828,49.988,92.243", Scale="1,1,1") {
    rotationMode=BMM_Additive
    rotationSpace=BCS_BoneSpace
    scaleMode=BMM_Additive
    scaleSpace=BCS_BoneSpace
    translationMode=BMM_Additive
    translationSpace=BCS_BoneSpace
    alphaScaleBiasClamp {bInterpResult=true}
    boneToModify {boneName=dyn_r_hair_b_02}
    rotation=<rot 37.828 49.988 92.243>
    translation=<-0.958 0.156 -0.731>
   }
   n44 = "Transform (Modify) Bone - Bone: dyn_l_hair_a_01"(ComponentPose=@n43, Alpha=1, Translation="0,0,0", Rotation="0.000,-55.000,-0.000", Scale="1,1,1") {
    rotationMode=BMM_Additive
    rotationSpace=BCS_BoneSpace
    scaleMode=BMM_Additive
    scaleSpace=BCS_BoneSpace
    translationMode=BMM_Additive
    translationSpace=BCS_BoneSpace
    alphaScaleBiasClamp {bInterpResult=true}
    boneToModify {boneName=dyn_l_hair_a_01}
    rotation=<rot 0 -54.9999 0>
   }
   n45 = "Transform (Modify) Bone - Bone: dyn_r_hair_b_01"(ComponentPose=@n44, Alpha=1, Translation="2.281,0.491,-1.466", Rotation="28.440,75.475,49.882", Scale="1,1,1") {
    rotationMode=BMM_Additive
    rotationSpace=BCS_BoneSpace
    scaleMode=BMM_Additive
    scaleSpace=BCS_BoneSpace
    translationMode=BMM_Additive
    translationSpace=BCS_BoneSpace
    alphaScaleBiasClamp {bInterpResult=true}
    boneToModify {boneName=dyn_r_hair_b_01}
    rotation=<rot 28.44 75.475 49.882>
    translation=<2.281 0.491 -1.466>
   }
   n46 = "Transform (Modify) Bone - Bone: dyn_r_hair_b_02"(ComponentPose=@n45, Alpha=1, Translation="0,0,0", Rotation="-48.944,-77.415,16.202", Scale="1,1,1") {
    rotationMode=BMM_Additive
    rotationSpace=BCS_BoneSpace
    scaleMode=BMM_Additive
    scaleSpace=BCS_BoneSpace
    translationMode=BMM_Additive
    translationSpace=BCS_BoneSpace
    alphaScaleBiasClamp {bInterpResult=true}
    boneToModify {boneName=dyn_r_hair_b_02}
    rotation=<rot -48.944 -77.415 16.202>
   }
   n47 = "Transform (Modify) Bone - Bone: dyn_l_hair_b_02"(ComponentPose=@n46, Alpha=1, Translation="-3.904,0.130,0.000", Rotation="0,0,0", Scale="1,1,1") {
    rotationMode=BMM_Additive
    rotationSpace=BCS_BoneSpace
    scaleMode=BMM_Additive
    scaleSpace=BCS_BoneSpace
    translationMode=BMM_Additive
    translationSpace=BCS_BoneSpace
    alphaScaleBiasClamp {bInterpResult=true}
    boneToModify {boneName=dyn_l_hair_b_02}
    translation=<-3.904 0.13>
   }
   n48 = "Transform (Modify) Bone - Bone: dyn_r_hair_b_01"(ComponentPose=@n47, Alpha=1, Translation="0,0,0", Rotation="-0.000,-20.000,-0.000", Scale="1,1,1") {
    rotationMode=BMM_Additive
    rotationSpace=BCS_BoneSpace
    scaleMode=BMM_Additive
    scaleSpace=BCS_BoneSpace
    translationMode=BMM_Additive
    translationSpace=BCS_BoneSpace
    alphaScaleBiasClamp {bInterpResult=true}
    boneToModify {boneName=dyn_r_hair_b_01}
    rotation=<rot 0 -20 0>
   }
   n49 = "Transform (Modify) Bone - Bone: dyn_r_hair_a_01"(ComponentPose=@n48, Alpha=1, Translation="0,0,0", Rotation="0.001,5.000,-0.000", Scale="1,1,1") {
    rotationMode=BMM_Additive
    rotationSpace=BCS_BoneSpace
    scaleMode=BMM_Additive
    scaleSpace=BCS_BoneSpace
    translationMode=BMM_Additive
    translationSpace=BCS_BoneSpace
    alphaScaleBiasClamp {bInterpResult=true}
    boneToModify {boneName=dyn_r_hair_a_01}
    rotation=<rot 0.0009 4.9999 -0.0001>
   }
   n50 = "Transform (Modify) Bone - Bone: dyn_r_hair_a_01"!(ComponentPose=@n49, bAlphaBoolEnabled:bool=True, Translation="0,0,0", Rotation="5,9.824,1.338", Scale="1,1,1") -> (Pose) {
    alphaInputType=Bool
    rotationMode=BMM_Additive
    rotationSpace=BCS_BoneSpace
    scaleMode=BMM_Additive
    scaleSpace=BCS_BoneSpace
    translationMode=BMM_Additive
    translationSpace=BCS_BoneSpace
    alphaScaleBiasClamp {bInterpResult=true}
    boneToModify {boneName=dyn_r_hair_a_01}
    rotation=<rot 5 9.824 1.338>
   }
   n51 = "Transform (Modify) Bone - Bone: dyn_r_hair_b_01"!(ComponentPose=@n50, bAlphaBoolEnabled:bool=True, Translation="0,0,0", Rotation="10,10,0", Scale="1,1,1") -> (Pose) {
    alphaInputType=Bool
    rotationMode=BMM_Additive
    rotationSpace=BCS_BoneSpace
    scaleMode=BMM_Additive
    scaleSpace=BCS_BoneSpace
    translationMode=BMM_Additive
    translationSpace=BCS_BoneSpace
    alphaScaleBiasClamp {bInterpResult=true}
    boneToModify {boneName=dyn_r_hair_b_01}
    rotation {pitch=10, yaw=10}
   }
   n52 = Anim Dynamics - Bone: dyn_r_hair_a_01!(ComponentPose=@n51, Alpha=1, GravityOverride:Vector="0,0,-980", ExternalForce="0,0,0") -> (Pose) {
    angularDampingOverride=1
    angularSpringConstant=10000
    bAngularSpring=true
    bOverrideAngularDamping=true
    bUseGravityOverride=true
    gravityScale=0.35
    simulationSpace=RootRelative
    boundBone {boneName=dyn_r_hair_a_01}
    componentAppliedLinearAccClamp=<5000 5000 5000>
    componentLinearAccScale=<0.5 0.5 0.5>
    componentLinearVelScale=<0.5 0.5 0.5>
    gravityOverride {z=-980}
    physicsBodyDefinitions [{
      boundBone {boneName=dyn_r_hair_a_01}
      boxExtents=<15 20 20>
      constraintSetup {angularLimitsMax=<0 0 2>, angularLimitsMin=<0 0 -2>, angularTarget=<1 0 0>}
      localJointOffset=<3 0 0>
     }]
   }
   n53 = Anim Dynamics - Bone: dyn_r_hair_a_02!(ComponentPose=@n52, Alpha=1, GravityOverride:Vector="0,0,-980", ExternalForce="0,0,0") -> (Pose) {
    angularDampingOverride=1
    angularSpringConstant=10000
    bAngularSpring=true
    bOverrideAngularDamping=true
    bUseGravityOverride=true
    gravityScale=0.35
    simulationSpace=RootRelative
    boundBone {boneName=dyn_r_hair_a_02}
    componentAppliedLinearAccClamp=<5000 5000 5000>
    componentLinearAccScale=<0.5 0.5 0.5>
    componentLinearVelScale=<0.5 0.5 0.5>
    gravityOverride {z=-980}
    physicsBodyDefinitions [{
      boundBone {boneName=dyn_r_hair_a_02}
      boxExtents=<15 20 20>
      constraintSetup {angularLimitsMax=<0 0 2>, angularLimitsMin=<0 0 -2>, angularTarget=<1 0 0>}
      localJointOffset=<3 0 0>
     }]
   }
   n54 = Component To Local(ComponentPose=@n53)
   n55 = Anim Dynamics - Bone: dyn_r_feather_02!(
    ComponentPose=Local To Component(LocalPose=Blend Poses by bool(BlendPose_0=@n27, BlendPose_1=@n54, BlendTime_0=1, BlendTime_1=1, bActiveValue=@n28))
    Alpha=1
    GravityOverride:Vector="0,0,-980"
    ExternalForce="0,0,0"
   ) -> (Pose) {
    angularDampingOverride=1
    angularSpringConstant=5000
    bAngularSpring=true
    bOverrideAngularDamping=true
    bUseGravityOverride=true
    gravityScale=0.35
    simulationSpace=RootRelative
    boundBone {boneName=dyn_r_feather_02}
    componentAppliedLinearAccClamp=<1500 1500 1500>
    gravityOverride {z=-980}
    physicsBodyDefinitions [{
      boundBone {boneName=dyn_r_feather_02}
      boxExtents=<15 15 15>
      constraintSetup {angularLimitsMax=<0 2 2>, angularLimitsMin=<0 -2 -2>, angularTarget=<1 0 0>}
      localJointOffset=<2 0 0>
     }]
    simSpaceSettings {maxAngularAcceleration=1500, maxAngularVelocity=2000}
   }
   n56 = Anim Dynamics - Bone: dyn_l_feather_02!(ComponentPose=@n55, Alpha=1, GravityOverride:Vector="0,0,-980", ExternalForce="0,0,0") -> (Pose) {
    angularDampingOverride=1
    angularSpringConstant=5000
    bAngularSpring=true
    bOverrideAngularDamping=true
    bUseGravityOverride=true
    gravityScale=0.35
    simulationSpace=RootRelative
    boundBone {boneName=dyn_l_feather_02}
    componentAppliedLinearAccClamp=<1500 1500 1500>
    gravityOverride {z=-980}
    physicsBodyDefinitions [{
      boundBone {boneName=dyn_l_feather_02}
      boxExtents=<15 15 15>
      constraintSetup {angularLimitsMax=<0 2 2>, angularLimitsMin=<0 -2 -2>, angularTarget=<1 0 0>}
      localJointOffset=<2 0 0>
     }]
    simSpaceSettings {maxAngularAcceleration=1500, maxAngularVelocity=2000}
   }
   n57 = Anim Dynamics - Bone: dyn_f_hair!(ComponentPose=@n56, Alpha=1, GravityOverride:Vector="0,0,-980", ExternalForce="0,0,0") -> (Pose) {
    angularDampingOverride=1
    angularSpringConstant=5000
    bAngularSpring=true
    bOverrideAngularDamping=true
    bUseGravityOverride=true
    gravityScale=0.35
    simulationSpace=RootRelative
    boundBone {boneName=dyn_f_hair}
    componentLinearAccScale=<0.5 0.5 0.5>
    componentLinearVelScale=<0.5 0.5 0.5>
    gravityOverride {z=-980}
    physicsBodyDefinitions [{
      boundBone {boneName=dyn_f_hair}
      boxExtents=<15 20 20>
      constraintSetup {angularLimitsMax=<0 0 5>, angularLimitsMin=<0 -15 -5>, angularTarget=<1 0 0>}
      localJointOffset=<2 0 0>
     }]
   }
   n58 = Anim Dynamics - Bone: dyn_r_hair_01!(ComponentPose=@n57, Alpha=1, GravityOverride:Vector="0,0,-980", ExternalForce="0,0,0") -> (Pose) {
    angularDampingOverride=1
    angularSpringConstant=10000
    bAngularSpring=true
    bOverrideAngularDamping=true
    bUseGravityOverride=true
    gravityScale=0.35
    simulationSpace=RootRelative
    boundBone {boneName=dyn_r_hair_01}
    componentAppliedLinearAccClamp=<1500 1500 1500>
    componentLinearAccScale=<0.5 0.5 0.5>
    componentLinearVelScale=<0.5 0.5 0.5>
    gravityOverride {z=-980}
    physicsBodyDefinitions [{
      boundBone {boneName=dyn_r_hair_01}
      boxExtents=<15 20 20>
      constraintSetup {angularLimitsMax=<0 0 5>, angularLimitsMin=<0 -2 -1>, angularTarget=<1 0 0>}
      localJointOffset=<2 0 0>
     }]
   }
   n59 = Anim Dynamics - Bone: dyn_l_hair_01!(ComponentPose=@n58, Alpha=1, GravityOverride:Vector="0,0,-980", ExternalForce="0,0,0") -> (Pose) {
    angularDampingOverride=1
    angularSpringConstant=10000
    bAngularSpring=true
    bOverrideAngularDamping=true
    bUseGravityOverride=true
    gravityScale=0.35
    simulationSpace=RootRelative
    boundBone {boneName=dyn_l_hair_01}
    componentAppliedLinearAccClamp=<1500 1500 1500>
    componentLinearAccScale=<0.5 0.5 0.5>
    componentLinearVelScale=<0.5 0.5 0.5>
    gravityOverride {z=-980}
    physicsBodyDefinitions [{
      boundBone {boneName=dyn_l_hair_01}
      boxExtents=<15 20 20>
      constraintSetup {angularLimitsMax=<0 0 5>, angularLimitsMin=<0 -2 -1>, angularTarget=<1 0 0>}
      localJointOffset=<2 0 0>
     }]
   }
   n60 = "Transform (Modify) Bone - Bone: dyn_l_hair_a_01"(ComponentPose=@n59, Alpha=1, Translation="-3,0,0", Rotation="0,0,0", Scale="1,1,1") {
    rotationMode=BMM_Additive
    rotationSpace=BCS_BoneSpace
    scaleMode=BMM_Additive
    scaleSpace=BCS_BoneSpace
    translationMode=BMM_Additive
    translationSpace=BCS_BoneSpace
    alphaScaleBiasClamp {bInterpResult=true}
    boneToModify {boneName=dyn_l_hair_a_01}
    translation {x=-3}
   }
   n61 = "Transform (Modify) Bone - Bone: dyn_r_hair_a_01"!(ComponentPose=@n60, Alpha:float=1, Translation="-3,0,0", Rotation="0,0,0", Scale="1,1,1") -> (Pose) {
    rotationMode=BMM_Additive
    rotationSpace=BCS_BoneSpace
    scaleMode=BMM_Additive
    scaleSpace=BCS_BoneSpace
    translationMode=BMM_Additive
    translationSpace=BCS_BoneSpace
    alphaScaleBiasClamp {bInterpResult=true}
    boneToModify {boneName=dyn_r_hair_a_01}
    translation {x=-3}
   }
   n62 = "Transform (Modify) Bone - Bone: dyn_l_hair_b_01"!(ComponentPose=@n61, bAlphaBoolEnabled:bool=True, Translation="0,0,0", Rotation="9.827,15.169,1.313", Scale="1,1,1") -> (Pose) {
    alphaInputType=Bool
    rotationMode=BMM_Additive
    rotationSpace=BCS_BoneSpace
    scaleMode=BMM_Additive
    scaleSpace=BCS_BoneSpace
    translationMode=BMM_Additive
    translationSpace=BCS_BoneSpace
    alphaScaleBiasClamp {bInterpResult=true}
    boneToModify {boneName=dyn_l_hair_b_01}
    rotation=<rot 9.8269 15.1691 1.313>
   }
   n63 = "Transform (Modify) Bone - Bone: dyn_r_hair_b_01"(ComponentPose=@n62, bAlphaBoolEnabled=True, Translation="0,0,0", Rotation="9.847,15.151,1.753", Scale="1,1,1") {
    alphaInputType=Bool
    rotationMode=BMM_Additive
    rotationSpace=BCS_BoneSpace
    scaleMode=BMM_Additive
    scaleSpace=BCS_BoneSpace
    translationMode=BMM_Additive
    translationSpace=BCS_BoneSpace
    alphaScaleBiasClamp {bInterpResult=true}
    boneToModify {boneName=dyn_r_hair_b_01}
    rotation=<rot 9.847 15.1509 1.753>
   }
   n64 = "Transform (Modify) Bone - Bone: dyn_m_hair_02"!(ComponentPose=@n63, bAlphaBoolEnabled:bool=True, Translation="0,0,0", Rotation="25.000,0.000,0.000", Scale="1,1,1") -> (Pose) {
    alphaInputType=Bool
    rotationMode=BMM_Additive
    rotationSpace=BCS_BoneSpace
    scaleMode=BMM_Additive
    scaleSpace=BCS_BoneSpace
    translationMode=BMM_Additive
    translationSpace=BCS_BoneSpace
    alphaScaleBiasClamp {bInterpResult=true}
    boneToModify {boneName=dyn_m_hair_02}
    rotation=<rot 24.9996 0 0.0001>
   }
   n65 = "Transform (Modify) Bone - Bone: dyn_r_hair_b_01"(ComponentPose=@n64, bAlphaBoolEnabled=True, Translation="0,0,0", Rotation="-0.000,-20.000,0.000", Scale="1,1,1") {
    alphaInputType=Bool
    rotationMode=BMM_Additive
    rotationSpace=BCS_BoneSpace
    scaleMode=BMM_Additive
    scaleSpace=BCS_BoneSpace
    translationMode=BMM_Additive
    translationSpace=BCS_BoneSpace
    alphaScaleBiasClamp {bInterpResult=true}
    boneToModify {boneName=dyn_r_hair_b_01}
    rotation=<rot -0.0001 -20.0001 0>
   }
   n66 = "Transform (Modify) Bone - Bone: dyn_l_hair_b_01"(ComponentPose=@n65, bAlphaBoolEnabled=True, Translation="0,0,0", Rotation="4.981,-5.019,-0.437", Scale="1,1,1") {
    alphaInputType=Bool
    rotationMode=BMM_Additive
    rotationSpace=BCS_BoneSpace
    scaleMode=BMM_Additive
    scaleSpace=BCS_BoneSpace
    translationMode=BMM_Additive
    translationSpace=BCS_BoneSpace
    alphaScaleBiasClamp {bInterpResult=true}
    boneToModify {boneName=dyn_l_hair_b_01}
    rotation=<rot 4.9809 -5.0189 -0.437>
   }
   Save cached pose 'DYN'(Pose=Component To Local(ComponentPose=@n66)) {cacheName=DYN}
  ]}
 EventGraph {nodes [
   Event BlueprintUpdateAnimation()
   > n1 = Sequence!(execute) -> (then_0, then_1, then_2:Exec, then_3:Exec, then_4:Exec, then_5:Exec, then_6:Exec)
   > n2 = "Gravity Override State Function (Parameter Asset) (Nativized)"(
    CharacterPartAnimInstance=Self-Reference()
    GameTimeInSeconds=GetGameTimeInSeconds()
    Parameters=&/F_MED_FuzzyGlam_FaceAcc_Parameters
   )
   > Set GravityOverride(GravityOverride=@n2)
   > Set RBAN Settings Out(RBAN Settings Out=@n2.OutRigidBodyAnimNodeParameters)
   Set CrouchWalk(execute=@n1.then_1, CrouchWalk:bool=AND Boolean(A=Get bIsCrouching(), B=Get bIsMoving2D())) -> (Output_Get:bool)
   n3 = Get bIsMoving2D()
   Set CrouchTargeting(
    execute=@n1.then_2
    CrouchTargeting:bool=AND Boolean!(A=Get bIsCrouching(), B=Get bIsTargeting(), C:bool=NOT Boolean(A=@n3)) -> (ReturnValue)
   ) -> (Output_Get:bool)
   n4 = Get bIsCrouching()
   n5 = Get bIsTargeting()
   > Set InPlaceMove(InPlaceMove:double=SelectFloat(A=Get Head Up_New() -> (Head Up_New:double), B=0, bPickA=AND Boolean!(A=@n5, B=@n4) -> (ReturnValue))) -> (Output_Get:double)
   n6 = float + float(A=Get Head_Turn_L() -> (Head_Turn_L:double), B=Get Head Up_New())
   > Set Head_L_UP(Head_L_UP:double=@n6) -> (Output_Get:double)
   > Set Emote_HeadRot(Emote_HeadRot:double="To Float (Boolean)"(InBool=Get bIsPlayingEmote())) -> (Output_Get:double)
   n7 = Set Head Up(
    execute=@n1.then_3
    Head Up:double=GetAlphaFromReferenceTransformDelta(SkeletalMeshComponent=Get Skel Mesh(), BoneName=spine_05, BaseName=head, InputRange=(X=0,Y=90), TwistAxis="0,0,1")
   ) -> (Output_Get:double)
   > Set CrouchingHeadUP(CrouchingHeadUP:double=SelectFloat(A=@n7, B=0, bPickA=Get CrouchTargeting() -> (CrouchTargeting:bool))) -> (Output_Get:double)
   n8 = NOT Boolean(A=Get bIsCrouching())
   n9 = Get bIsTargeting()
   > Set AimUp(AimUp:double=SelectFloat(A=@n7, B=0, bPickA=AND Boolean(A=@n8, B=@n9))) -> (Output_Get:double)
   n10 = NOT Boolean(A=Get bIsCrouching())
   n11 = Get bIsTargeting()
   > Set Head Up_Target(Head Up_Target:double=SelectFloat(A=Get Head Up() -> (Head Up:double), B=0, bPickA=AND Boolean(A=@n11, B=@n10))) -> (Output_Get:double)
   n12 = Get bIsCrouching()
   > Set IsCrouching(IsCrouching:double="To Float (Boolean)"(InBool=@n12)) -> (Output_Get:double)
   n13 = Get bIsTargeting()
   > Set IsTragetting(IsTragetting:double="To Float (Boolean)"(InBool=@n13)) -> (Output_Get:double)
   n14 = SelectFloat(A=Get Head Down() -> (Head Down:double), B=0, bPickA=@n13)
   > Set Head Down Target(Head Down Target:double="Absolute (Float)"(A=float - float(A=1, B=@n14))) -> (Output_Get:double)
   n15 = BreakTransform(InTransform=GetDeltaTransformFromRefPose(self=GetOwnerSkeletalMeshComponent(), BoneName=head, BaseName=spine_05))
   > Set Head Up_New(Head Up_New:double=MapRangeClamped(Value=BreakRotator(InRot=@n15.Rotation).Yaw, InRangeA=0, InRangeB=-75)) -> (Output_Get:double)
   n16 = AND Boolean!(A=Get bIsCrouching(), B=Get bIsTargeting(), C:bool=Get bIsMoving2D()) -> (ReturnValue)
   n17 = Get Head Up_New()
   > Set CrouchHeadsup(CrouchHeadsup:double="To Float (Boolean)"(InBool="Equal (Float)"(A=@n17, B:bool=@n16))) -> (Output_Get:double)
   n18 = Get bIsPlayingEmote()
   > Set IsplayingEmote(IsplayingEmote:double="To Float (Boolean)"(InBool=@n18)) -> (Output_Get:double)
   > Set IsplayingEmote_Not(IsplayingEmote_Not:double="To Float (Boolean)"(InBool=NOT Boolean(A=@n18))) -> (Output_Get:double)
   n19 = AND Boolean(A=Get bIsTargeting(), B=Get bIsCrouching(), C=Get bIsMoving2D())
   n20 = Get Head Up_New()
   > Set CrouchingHeadUP_Walk(CrouchingHeadUP_Walk:double="To Float (Boolean)"(InBool="Equal (Float)"(A=@n20, B=@n19))) -> (Output_Get:double)
   n21 = Get bIsTargeting()
   > Set HeadsUp_Tagetting(HeadsUp_Tagetting:double="To Float (Boolean)"(InBool=@n21)) -> (Output_Get:double)
   n22 = AND Boolean!(A=Get bIsTargeting(), B=Get bIsCrouching()) -> (ReturnValue)
   > Set CrouchHeadUpWalk(CrouchHeadUpWalk:double="To Float (Boolean)"(InBool=@n22)) -> (Output_Get:double)
   > Set ISCrouchWalk(ISCrouchWalk:bool=@n22) -> (Output_Get:bool)
   n23 = NOT Boolean(A=Get bIsMoving2D())
   n24 = Get bIsCrouching()
   n25 = Get bIsTargeting()
   > Set CrouchHeadUpTargetting(CrouchHeadUpTargetting:double="To Float (Boolean)"(InBool=AND Boolean!(A=@n25, B=@n24, C:bool=@n23) -> (ReturnValue))) -> (Output_Get:double)
   n26 = NOT Boolean(A=Get bIsTargeting())
   n27 = NOT Boolean(A=Get bIsCrouching())
   > Set Head Up_Only(Head Up_Only:double=SelectFloat(A=Get Head Up_New(), B=0, bPickA=AND Boolean!(A=@n26, B=@n27) -> (ReturnValue))) -> (Output_Get:double)
   n28 = NOT Boolean(A=Get bIsTargeting())
   n29 = Get bIsCrouching()
   > Set Head Up_Only_Crouch(Head Up_Only_Crouch:double=SelectFloat(A=Get Head Up_New(), B=0, bPickA=AND Boolean(A=@n28, B=@n29))) -> (Output_Get:double)
   Set RBEnabledFloat(execute=@n1.then_6, RBEnabledFloat:double="To Float (Boolean)"(InBool=Get bRigidBodyEnabled())) -> (Output_Get:double)
   n30 = Get Skel Mesh()
   Set Head Up(
    execute=@n1.then_4
    Head Up=GetAlphaFromReferenceTransformDelta(SkeletalMeshComponent=@n30, BoneName=spine_05, BaseName=head, InputRange=(X=-13,Y=-51), TwistAxis="0,0,1")
   )
   > Set Head Down(Head Down:double=GetAlphaFromReferenceTransformDelta(SkeletalMeshComponent=@n30, BoneName=spine_05, BaseName="head ", InputRange=(X=0,Y=-90), OutputRange=(X=0,Y=-1), TwistAxis="0,0,1")) -> (Output_Get:double)
   > Set Head_Turn_L(Head_Turn_L:double=GetAlphaFromReferenceTransformDelta(SkeletalMeshComponent=@n30, BoneName=spine_05, BaseName=head, DeltaComponentType=EulerX, InputRange=(X=0,Y=-22), TwistAxis="1,0,0")) -> (Output_Get:double)
   > Set Head_Turn_R(Head_Turn_R:double=GetAlphaFromReferenceTransformDelta(SkeletalMeshComponent=@n30, BoneName=spine_05, BaseName=head, DeltaComponentType=EulerX, InputRange=(X=0,Y=22), TwistAxis="1,0,0")) -> (Output_Get:double)
   n31 = CheckForBackbling()
   Set IsWearingBackbling(execute=@n1.then_5, IsWearingBackbling:bool=OR Boolean!(A=@n31, B=@n31.bIsBackblingACape, C:bool=@n31.bIsBackblingLong) -> (ReturnValue)) -> (Output_Get:bool)
   > Set Is Backbling A Cape(Is Backbling A Cape:bool=@n31.bIsBackblingACape) -> (Output_Get:bool)
   > Set is Wear a Cape(is Wear a Cape:bool=AND Boolean(A=@n31, B=@n31.bIsBackblingLong)) -> (Output_Get:bool)
   n32 = Event OnCharacterCustomizationLoaded()
   > Set Body(Body:Skeletal Mesh Component Budgeted Object Reference=Get Array Item with EFort Custom Part Type(Array=@n32.CustomMeshParts)) -> (Output_Get:Skeletal Mesh Component Budgeted Object Reference)
   Event BlueprintInitializeAnimation()
   > Set Skel Mesh(Skel Mesh=GetOwnerSkeletalMeshComponent())
   TryGetPawnOwner()
   SelectFloat(A=@n20, B=0, bPickA=@n21)
   NOT Boolean(A=@n29)
   SelectFloat(A=@n6, B=0, bPickA=Get bIsTargeting())
   NOT Boolean(A=Get bIsMoving2D())
   AND Boolean!(A=@n12, B=Get bIsTargeting(), C:bool=Get bIsMoving2D()) -> (ReturnValue)
  ]}
]
variables [
 AimUp {type=number}
 AlphaJumping {type=number}
 Body {type=SkeletalMeshComponentBudgeted}
 CrouchAlpha {type=number}
 CrouchHeadsup {type=number}
 CrouchHeadUpTargetting {type=number}
 CrouchHeadUpWalk {type=number}
 CrouchingHeadUP {type=number}
 CrouchingHeadUP_Bool {type=boolean}
 CrouchingHeadUP_Walk {type=number}
 CrouchTargeting {type=boolean}
 CrouchWalk {type=boolean}
 Emote_HeadRot {type=number}
 GravityOverride {type=Vector}
 Head Down {type=number}
 Head Down Target {type=number}
 Head Up {type=number}
 Head Up_Modified {type=number}
 Head Up_New {type=number}
 Head Up_Only {type=number}
 Head Up_Only_Crouch {type=number}
 Head Up_Target {type=number}
 Head_L_UP {type=number}
 Head_Tilt_R {type=number}
 Head_Tilt_R_Bool {type=boolean}
 Head_Turn_L {type=number}
 Head_Turn_L_B {type=boolean}
 Head_Turn_R {type=number}
 HeadsUp_Tagetting {type=number}
 HeadUP Left {type=number}
 HeadUP Left_Float {type=number}
 HeadUP LeftTilt {type=boolean}
 InPlaceMove {type=number}
 Is Backbling A Cape {type=boolean}
 is Wear a Cape {type=boolean}
 IsCrouching {type=number}
 ISCrouchWalk {type=boolean}
 IsplayingEmote {type=number}
 IsplayingEmote_Not {type=number}
 IsTragetting {type=number}
 IsWearingBackbling {type=boolean}
 LookUpCrouchAlpha {type=number}
 LookUpPoseAlpha {type=number}
 RBAN Settings Out {type=FortRigidBodyAnimNodeParameters}
 RBEnabledFloat {type=number}
 Skel Mesh {type=SkeletalMeshComponent}
]
