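# Physics asset for the FuzzyGlam medium female FaceAcc_CapFar cosmetic part, covering its left and right dynamic feather bones.
# Each side is a two-bone chain of same-sized capsules, with the first feather bone constrained to a zero-size head capsule and the second to the first.
# Constraints set very small swing limits, slightly wider at the second bone, and zero twist limits.
# All four feather capsules share the left first feather's shape name.
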PhysicsAsset /BRCosmetics/Characters/Player/Female/Medium/Bodies/F_MED_FuzzyGlam/Meshes/Parts/F_MED_FuzzyGlam_FaceAcc_CapFar_Physics
bodies {
 dyn_l_feather_01 [{
   shapeType=Capsule
   length=8.1913
   radius=3.1838
   shapeName=dyn_l_feather_01_capsule
   center=<0.6768 0.0138 -0.0305>
   rotation=<rot -0.3848 -13.1989 -92.1829>
  }]
 dyn_l_feather_02 [{
   shapeType=Capsule
   length=8.1913
   radius=3.1838
   shapeName=dyn_l_feather_01_capsule
   center=<1.742 -0.236 -0.0379>
   rotation=<rot -0.3848 -13.1989 -92.1829>
  }]
 dyn_r_feather_01 [{
   shapeType=Capsule
   length=8.1913
   radius=3.1838
   shapeName=dyn_l_feather_01_capsule
   center=<-0.1762 -0.2562 0.2958>
   rotation=<rot 0.7579 16.7798 -92.083>
  }]
 dyn_r_feather_02 [{
   shapeType=Capsule
   length=8.1913
   radius=3.1838
   shapeName=dyn_l_feather_01_capsule
   center=<0.8041 1.2843 0.0189>
   rotation=<rot 0.7579 16.7798 -92.083>
  }]
 head [{shapeType=Capsule, length=0, radius=0, shapeName=head_capsule}]
}
constraints [
 {bone1Name=dyn_l_feather_01, bone2Name=head, swing1LimitDegrees=0.5, swing2LimitDegrees=0.3, twistLimitDegrees=0}
 {bone1Name=dyn_l_feather_02, bone2Name=dyn_l_feather_01, swing1LimitDegrees=1, swing2LimitDegrees=2, twistLimitDegrees=0}
 {bone1Name=dyn_r_feather_01, bone2Name=head, swing1LimitDegrees=0.5, swing2LimitDegrees=0.3, twistLimitDegrees=0}
 {bone1Name=dyn_r_feather_02, bone2Name=dyn_r_feather_01, swing1LimitDegrees=1, swing2LimitDegrees=2, twistLimitDegrees=0}
]
