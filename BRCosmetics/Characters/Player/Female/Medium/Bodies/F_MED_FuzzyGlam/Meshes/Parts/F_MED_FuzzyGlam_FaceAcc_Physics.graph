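# Physics asset for the FuzzyGlam medium female face accessory, covering three two-bone dynamic hair chains (left, middle, right) on the head.
# Capsule bodies cover the hair, head, neck, upper spine and clavicles, with free angular joints between the body bones.
# Hair chains have swing limits and zero twist, and tighter lateral constraints tie the side chains to the middle one.
# The right-to-middle links also carry duplicate free-motion constraints.
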PhysicsAsset /BRCosmetics/Characters/Player/Female/Medium/Bodies/F_MED_FuzzyGlam/Meshes/Parts/F_MED_FuzzyGlam_FaceAcc_Physics
bodies {
 clavicle_l [{shapeType=Capsule, length=10, radius=5, shapeName=clavicle_l_capsule, center=<5 -2 -0.5>, rotation=<rot 80 0 0>}]
 clavicle_r [{shapeType=Capsule, length=12, radius=5.1, shapeName=clavicle_r_capsule, center=<-5 2 0.5>, rotation=<rot 80 0 0>}]
 dyn_l_hair_b_01 [{shapeType=Capsule, length=5, radius=4, shapeName=dyn_l_hair_b_01_capsule, center=<3 0 0>, rotation=<rot 90 0 0>}]
 dyn_l_hair_b_02 [{shapeType=Capsule, length=5, radius=3.8, shapeName=dyn_l_hair_b_02_capsule, center=<5 0 0>, rotation=<rot 90 0 0>}]
 dyn_m_hair_01 [{shapeType=Capsule, length=5, radius=4, shapeName=dyn_m_hair_01_capsule, center=<3 0 0>, rotation=<rot 90 0 0>}]
 dyn_m_hair_02 [{shapeType=Capsule, length=5, radius=3.8, shapeName=dyn_m_hair_02_capsule, center=<5 0 0>, rotation=<rot 90 0 0>}]
 dyn_r_hair_b_01 [{shapeType=Capsule, length=5, radius=4, shapeName=dyn_r_hair_b_01_capsule, center=<3 0 0>, rotation=<rot 90 0 0>}]
 dyn_r_hair_b_02 [{shapeType=Capsule, length=5, radius=3.9, shapeName=dyn_r_hair_b_02_capsule, center=<5 2 1>, rotation=<rot 72.806 -155.2388 -180>}]
 head [{shapeType=Capsule, length=5, radius=3, shapeName=head_capsule}]
 neck_01 [{shapeType=Capsule, length=5, radius=3, shapeName=neck_01_capsule}]
 neck_02 [{shapeType=Capsule, length=5, radius=3, shapeName=neck_02_capsule}]
 spine_05 [{shapeType=Capsule, length=8.3804, radius=12, shapeName=spine_05_capsule, center=<-0.8543 3.0903 0>}]
}
constraints [
 {
  bone1Name=neck_02
  bone2Name=neck_01
  swing1LimitDegrees=45
  swing1Motion=Free
  swing2LimitDegrees=45
  swing2Motion=Free
  twistLimitDegrees=45
  twistMotion=Free
 }
 {
  bone1Name=head
  bone2Name=neck_02
  swing1LimitDegrees=45
  swing1Motion=Free
  swing2LimitDegrees=45
  swing2Motion=Free
  twistLimitDegrees=45
  twistMotion=Free
 }
 {
  bone1Name=neck_01
  bone2Name=spine_05
  swing1LimitDegrees=45
  swing1Motion=Free
  swing2LimitDegrees=45
  swing2Motion=Free
  twistLimitDegrees=45
  twistMotion=Free
 }
 {
  bone1Name=clavicle_l
  bone2Name=spine_05
  swing1LimitDegrees=45
  swing1Motion=Free
  swing2LimitDegrees=45
  swing2Motion=Free
  twistLimitDegrees=45
  twistMotion=Free
 }
 {
  bone1Name=clavicle_r
  bone2Name=spine_05
  swing1LimitDegrees=45
  swing1Motion=Free
  swing2LimitDegrees=45
  swing2Motion=Free
  twistLimitDegrees=45
  twistMotion=Free
 }
 {bone1Name=dyn_l_hair_b_01, bone2Name=head, swing1LimitDegrees=60, swing2LimitDegrees=80, twistLimitDegrees=0}
 {bone1Name=dyn_l_hair_b_02, bone2Name=dyn_l_hair_b_01, swing1LimitDegrees=50, swing2LimitDegrees=50, twistLimitDegrees=0}
 {bone1Name=dyn_m_hair_01, bone2Name=head, swing1LimitDegrees=70, swing2LimitDegrees=80, twistLimitDegrees=0}
 {bone1Name=dyn_m_hair_02, bone2Name=dyn_m_hair_01, swing1LimitDegrees=50, swing2LimitDegrees=50, twistLimitDegrees=0}
 {bone1Name=dyn_r_hair_b_01, bone2Name=head, swing1LimitDegrees=60, swing2LimitDegrees=80, twistLimitDegrees=0}
 {bone1Name=dyn_r_hair_b_02, bone2Name=dyn_r_hair_b_01, swing1LimitDegrees=50, swing2LimitDegrees=50, twistLimitDegrees=0}
 {bone1Name=dyn_l_hair_b_01, bone2Name=dyn_m_hair_01, swing1LimitDegrees=30, swing2LimitDegrees=15, twistLimitDegrees=0}
 {bone1Name=dyn_l_hair_b_02, bone2Name=dyn_m_hair_02, swing1LimitDegrees=30, swing2LimitDegrees=15, twistLimitDegrees=0}
 {bone1Name=dyn_r_hair_b_01, bone2Name=dyn_m_hair_01, swing1LimitDegrees=30, swing2LimitDegrees=15, twistLimitDegrees=0}
 {bone1Name=dyn_r_hair_b_02, bone2Name=dyn_m_hair_02, swing1LimitDegrees=30, swing2LimitDegrees=15, twistLimitDegrees=0}
 {
  bone1Name=dyn_r_hair_b_01
  bone2Name=dyn_m_hair_01
  swing1LimitDegrees=45
  swing1Motion=Free
  swing2LimitDegrees=45
  swing2Motion=Free
  twistLimitDegrees=45
  twistMotion=Free
 }
 {
  bone1Name=dyn_r_hair_b_02
  bone2Name=dyn_m_hair_02
  swing1LimitDegrees=45
  swing1Motion=Free
  swing2LimitDegrees=45
  swing2Motion=Free
  twistLimitDegrees=45
  twistMotion=Free
 }
]
